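# Niagara lightning effect for the RainbowRaid player emote, tinted by a shared system LightningColor.
# StaffLightning and Lightning continuously spawn short-lived, flickering mesh bolts animated through dynamic material parameters.
# Changes to User.BurstIndex trigger LightningSparks and Flash_Flare sprite bursts, with Flash_Flare merging three layers via spawn-group switches.
# Flare sizes and speeds adapt to front-end versus gameplay; Glow and Light emitters have no renderers.
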NiagaraSystem /BRCosmetics/Animation/Game/MainPlayer/Emotes/RainbowRaid/FX/NS_Emote_RainbowRaid_Lightning
dependencies {
 usedDataInterfaces [
  &NiagaraDataInterfaceCurve
  &NiagaraDataInterfacePlatformSet
  &NiagaraDataInterfaceSpriteRendererInfo
  &NiagaraDataInterfaceVector2DCurve
 ]
 usedDynamicInputs [&MakeCustomVectorFromBool, &SetBoolByFloatComparison]
 usedModules [
  &self:SystemSpawnScript.SystemScriptSource.SystemScriptGraph.NiagaraNodeAssignment_0.SetVariables_#1
  &self:SystemSpawnScript.SystemScriptSource.SystemScriptGraph.NiagaraNodeAssignment_2.SetVariables_#2
  &self:SystemSpawnScript.SystemScriptSource.SystemScriptGraph.NiagaraNodeAssignment_1.SetVariables_#3
  &SystemState
 ]
}
emitters {
 Flash_Flare {
  emitterUpdateScript [
   &EmitterState()
   /Game/Effects/Niagara/Modules/Logic/UpdateValueStateChange(InputFloatParam:NiagaraFloat=$User.BurstIndex, ParamType:ENiagaraValueStateChange=::NewEnumerator1)
   &self:TriggerSpawn(
    Spawn Count:NiagaraInt32=&RandomRangeInt(Minimum=15, Maximum=25)
    Trigger:NiagaraBool=$Emitter.UpdateValueStateChange.Changed
    Trigger01:NiagaraBool=$Emitter.UpdateValueStateChange.Changed
    Trigger02:NiagaraBool=$Emitter.UpdateValueStateChange.Changed
   ) {module=TriggerSpawn}
   &self:TriggerSpawn(
    Spawn Count:NiagaraInt32=&RandomRangeInt(Minimum=15, Maximum=25)
    Spawn Group=1
    Trigger:NiagaraBool=$Emitter.UpdateValueStateChange.Changed
    Trigger01:NiagaraBool=$Emitter.UpdateValueStateChange.Changed
    Trigger02:NiagaraBool=$Emitter.UpdateValueStateChange.Changed
   ) {module=TriggerSpawn001}
   &self:TriggerSpawn(
    Spawn Count:NiagaraInt32=&RandomRangeInt(Minimum=15, Maximum=25)
    Spawn Group=2
    Trigger:NiagaraBool=$Emitter.UpdateValueStateChange.Changed
    Trigger01:NiagaraBool=$Emitter.UpdateValueStateChange.Changed
    Trigger02:NiagaraBool=$Emitter.UpdateValueStateChange.Changed
   ) {module=TriggerSpawn002}
   &SpawnBurst_Instantaneous(
    Spawn Count=&MakeIntFromFloat(Float=&FN_PickFloatDependOnPlatform(
      TargetPlatform=DataInterface() {propertyValues {Platforms {qualityLevelMask=-29}}}
      TargetPlatformActive=&MakeFloatFromInt(Int=12)
      TargetPlatformNoActive=&MakeFloatFromInt(Int=30)
     ))
    Spawn Group=2
    Use Spawn Probability=false
    UseLoopCountLimit=false
   ) {enabled=false, module=SpawnBurst_Instantaneous002}
   &SpawnBurst_Instantaneous(Spawn Count=2, Spawn Group=1, Use Spawn Probability=false, UseLoopCountLimit=false) {enabled=false, module=SpawnBurst_Instantaneous001}
   &SpawnBurst_Instantaneous(Use Spawn Probability=false, UseLoopCountLimit=false) {enabled=false}
  ]
  particleSpawnScript [
   &PartitionParticles(Number Of Partitions=3) {enabled=false}
   &InitializeParticle(
    Color=&MergeHelper_ColorSwitch(
     0=&Multiply_LinearColorByFloat(Linear Color=$System.LightningColor, Float=3)
     1=$System.LightningColor
     2=&Multiply_LinearColorByFloat(Linear Color=$System.LightningColor, Float=3)
     3=?
     Index=$Emitter.SpawnGroup
     Num=::NewEnumerator1
    )
    Lifetime=&MergeHelper_FloatSwitch(
     0=0.1
     1=&RandomRangeFloat(Minimum=0.1, Maximum=0.15)
     2=&RandomRangeFloat(Minimum=0.1, Maximum=0.05)
     3=?
     Index=$Emitter.SpawnGroup
     Num=::NewEnumerator1
    )
    Position Offset=<7 0 10>
    Sprite Size=&Multiply_Vector2D_ByFloat(
     Float=0.4
     Vector2D=&MergeHelper_Vector2Switch(
      0=&FN_FrontEndVector2Modifier(Front End:Vector2f=<40 60>, Gameplay:Vector2f=<30 50>)
      1=&Multiply_Vector2D(
       A=&UniformRangedVector2D(Minimum=<20 30>, Maximum=<15 40>)
       B=&FN_FrontEndVector2Modifier(Front End:Vector2f=<0.5 0.7>, Gameplay:Vector2f=<1 1>)
      )
      2=&Multiply_Vector2D_ByFloat(
       Float=&Multiply_Float(
        A=&FN_FrontEndFloatModifier(Gameplay=1, Front End=0.45)
        B=&RandomRangeFloat(Minimum=4.5, Maximum=5)
       )
       Vector2D=<1 1.5>
      )
      3=?
      4=?
      Index=$Emitter.SpawnGroup
      Num=::NewEnumerator1
     )
    )
    Sprite Size Mode=::NewEnumerator4
    Sprite UV Mode=::NewEnumerator1
    UsePositionOffset=true
    Write Material Random=true
   )
   &ShapeLocation(
    Cylinder Height=&MergeHelper_FloatSwitch(0=0, 1=0, 2=80, 3=?, Index=$Emitter.SpawnGroup, Num=::NewEnumerator1)
    Cylinder Radius=&MergeHelper_FloatSwitch(0=0, 1=0, 2=2, 3=?, Index=$Emitter.SpawnGroup, Num=::NewEnumerator1)
    MaskSpawnGroup=false
    Override Seed=false
    Rotation Axis=$System.DirectionVector
    Rotation Mode=::NewEnumerator1
    Shape Primitive=::NewEnumerator1
    Surface Only=false
    UseOwnerScale=true
   )
   &CameraOffset(Camera Offset Amount=&MergeHelper_FloatSwitch(0=30, 1=25, 2=20, 3=?, Index=$Emitter.SpawnGroup, Num=::NewEnumerator1)) {enabled=false}
   &SubUVAnimation(
    End Frame Range Override=3
    Evaluation Type=::NewEnumerator0
    "Sprite Renderer "=DataInterface() {propertyValues {SpriteRenderer=&self:Flash_Flare_1.NiagaraSpriteRendererProperties_0}}
    SubUV Animation Mode:ENiagara_SubUVLookupModeV2=::NewEnumerator2
    UseEndFrame=true
    UseStartFrame=true
   )
   &DynamicMaterialParameters(
    Index 0 Param 1=0.2
    Index 0 Param 2=1
    Index 3 Param 1=1.5
    Index 3 Param 2=1
    Module.Index0Float0Write=true
    Module.Index0Float1Write=true
    Module.Index0Float2Write=true
    Module.Index0Float3Write=true
    Module.Index3Float0Write=true
    Module.Index3Float1Write=true
    Module.Index3Float2Write=true
    Module.Index3Float3Write=true
    Write Parameter Index 3=::NewEnumerator1
   )
   /SetVariables(Particles.DynamicParameterAlt:Vector4f=&MakeVector4(X=1, Y=1, Z=1, W=1)) {isSetParameters=true}
   &AddVelocity(
    Module.Use Falloff Distance=false
    Override Seed=false
    Velocity Mode=::NewEnumerator1
    Velocity Speed=&Multiply_Float(
     A=&MergeHelper_FloatSwitch(
      0=0
      1=20
      2=&RandomRangeFloat(Minimum=225, Maximum=300)
      3=?
      Index=$Emitter.SpawnGroup
      Num=::NewEnumerator1
     )
     B=$System.FrontEndScale
    )
   ) {module=AddVelocity001}
   &AddVelocity(
    Rotation Coordinate Space=::Simulation
    Velocity=$System.DirectionVector
    Velocity Speed Scale=&MergeHelper_FloatSwitch(
     0=&Multiply_Float(A=&RandomRangeFloat(Minimum=200, Maximum=300), B=$System.FrontEndScale)
     1=&Multiply_Float(A=100, B=$System.FrontEndScale)
     2=&Multiply_Float(A=&RandomRangeFloat(Minimum=300, Maximum=700), B=$System.FrontEndScale)
     3=?
     Index=$Emitter.SpawnGroup
     Num=::NewEnumerator1
    )
   )
   /SetVariables() {isSetParameters=true}
   /SetVariables(Particles.VisibilityTag=$Emitter.SpawnGroup) {isSetParameters=true}
  ]
  particleUpdateScript [
   &ParticleState()
   &UpdateMeshOrientation()
   &AccelerationForce(Acceleration=&MergeHelper_VectorSwitch(0=<0 0 0>, 1=<0 0 0>, 2=<0 0 -500>, 3=?, Index=$Particles.VisibilityTag, Num=::NewEnumerator1))
   &Drag(
    Drag=&MergeHelper_FloatSwitch(
     0=0
     1=0
     2=&RandomRangeFloat(Minimum=5, Maximum=8)
     3=?
     Index=$Particles.VisibilityTag
     Num=::NewEnumerator1
    )
    Rotational Drag:NiagaraFloat=&MergeHelper_FloatSwitch(0=0, 1=0, 2=1, 3=?, Index=$Particles.VisibilityTag, Num=::NewEnumerator1)
    Use Linear Drag=true
    Use Rotational Drag=true
   )
   &SolveForcesAndVelocity(Clamp Velocity=false, Limit Acceleration=false)
   &ScaleSpriteSizeBySpeed(
    Max Scale Factor=&MergeHelper_Vector2Switch(0=<1 1>, 1=<1 1>, 2=<0.2 5>, 3=?, 4=?, Index=$Particles.VisibilityTag, Num=::NewEnumerator1)
    Min Scale Factor=&MergeHelper_Vector2Switch(0=<1 1>, 1=<1 1>, 2=<0.2 0.2>, 3=?, 4=?, Index=$Particles.VisibilityTag, Num=::NewEnumerator1)
   )
   &ScaleSpriteSize(
    Scale Factor=&MergeHelper_Vector2Switch(
     0=&Vector2DFromCurve(
      Vector2Curve=DataInterface() {propertyValues {
        XCurve {keys [{tangentMode=RCTM_User, time=0, value=0}, <at 0.2 1.075>, {tangentMode=RCTM_User, time=1, value=-0.1}]}
        YCurve {keys [{tangentMode=RCTM_User, time=0, value=0}, <at 0.1 1.025>, {tangentMode=RCTM_User, time=1, value=0.8}]}
       }}
      ScaleCurve=<1 1>
     )
     1=&Vector2DFromCurve(
      Vector2Curve=DataInterface() {propertyValues {
        XCurve {keys [<at 0 1 RCIM_Cubic>,<at 1 0 RCIM_Cubic>]}
        YCurve {keys [{arriveTangent=3.9291, interpMode=RCIM_Cubic, leaveTangent=3.9291, tangentMode=RCTM_User, time=0, value=0}, <at 1 2 RCIM_Cubic>]}
       }}
      ScaleCurve=&Multiply_Vector2D_ByFloat(Float=1, Vector2D=<1.5 1>)
     )
     2=&Vector2DFromFloat(Value=&FloatFromCurve(FloatCurve=DataInterface() {propertyValues {Curve {keys [{arriveTangent=-6.725, interpMode=RCIM_Cubic, leaveTangent=-6.725, tangentMode=RCTM_User, time=0.65, value=1}, <at 1 0 RCIM_Cubic>]}}}))
     3=?
     4=?
     Index=$Particles.VisibilityTag
     Num=::NewEnumerator1
    )
    Scale Sprite Size Mode=::NewEnumerator2
   )
   &ScaleColor(
    Scale Alpha=&MergeHelper_FloatSwitch(
     0=1
     1=&FloatFromCurve(
      FloatCurve=DataInterface() {propertyValues {Curve {keys [<at 0 1 RCIM_Cubic>,<at 1 0>]}}}
      Scale Curve=&FNTimeOfDayScalar(Night Scale=0.5, Morning Scale=1, Evening Scale=0.5)
     )
     2=1
     3=?
     Index=$Particles.VisibilityTag
     Num=::NewEnumerator1
    )
    Scale RGB=&MergeHelper_VectorSwitch(
     0=<1 1 1>
     1=&VectorFromFloat(Value=&FloatFromCurve(FloatCurve=DataInterface() {propertyValues {Curve {keys [{arriveTangent=-15.8795, interpMode=RCIM_Cubic, leaveTangent=-15.8795, tangentMode=RCTM_User, time=0, value=10}, <at 1 1 RCIM_Cubic>]}}}))
     2=&VectorFromFloat(Value=&FloatFromCurve(
       FloatCurve=DataInterface() {propertyValues {Curve {keys [
           {interpMode=RCIM_Cubic, tangentMode=RCTM_User, time=0, value=1}
           <at 0.3 0.2>
           {interpMode=RCIM_Cubic, tangentMode=RCTM_User, time=1, value=0}
          ]}}}
       Scale Curve=10
      ))
     3=?
     Index=$Particles.VisibilityTag
     Num=::NewEnumerator1
    )
    ScaleA=true
    ScaleRGB=true
   )
   /SetVariables(
    Particles.GlowColor=&MakeLinearColorFromVectorAndFloat(
     "Vector (RGB)"=&MakeVectorFromLinearColorRGB(Color=$System.LightningColor)
     "Float (Alpha)"=&FloatFromCurve(FloatCurve=DataInterface() {propertyValues {Curve {keys [<at 0.5 1>,<at 1 0>]}}}, Scale Curve=0.1)
    )
    Particles.GlowSize=&Multiply_Vector2D(A=$Particles.SpriteSize, B=<8 2>)
   ) {isSetParameters=true}
  ]
  rendererClasses [&NiagaraSpriteRendererProperties]
  renderers [
   {rendererClass=&NiagaraSpriteRendererProperties}
   {rendererClass=&NiagaraSpriteRendererProperties, rendererIndex=1}
   {rendererClass=&NiagaraSpriteRendererProperties, rendererIndex=2}
   {rendererClass=&NiagaraSpriteRendererProperties, rendererIndex=3}
  ]
 }
 Glow {rendererClasses [], renderers []}
 Light {rendererClasses [], renderers []}
 Lightning {
  emitterUpdateScript [
   &EmitterState(Life Cycle Mode=::NewEnumerator1, Loop Behavior=::NewEnumerator1, Loop Duration Mode=::NewEnumerator1, UseLoopDelay=false)
   &SpawnRate(Spawn Probability=0.3, SpawnRate=5, Use Spawn Probability=true)
  ]
  particleSpawnScript [
   &InitializeParticle(
    Color=&Multiply_LinearColorByFloat(Linear Color=$System.LightningColor, Float=0.35)
    Lifetime=0.3
    Mesh Scale=&RandomRangeVector(Minimum=<0.2 0.2 0.2>, Maximum=<0.3 0.3 0.2>)
    Mesh Scale Mode=::NewEnumerator4
    UsePositionOffset=false
    Write Material Random=true
   )
   &ShapeLocation(
    Box Size=<5 5 100>
    MaskSpawnGroup=false
    Offset Mode=::NewEnumerator0
    Override Seed=false
    Shape Primitive=::NewEnumerator2
    Surface Only=false
    UseOwnerScale=true
   )
   &InitialMeshOrientation(
    Mesh Orientation Mode=::NewEnumerator2
    Orientation Vector=<-1 0 -1>
    Rotation=&RandomRangeVector(Minimum=<0.01 0 0>, Maximum=<0.03 0 0>)
    Use Rotation Vector=true
   )
  ]
  particleUpdateScript [
   &SolveForcesAndVelocity(Clamp Velocity=false, Limit Acceleration=false)
   &ParticleState()
   &ScaleMeshSize(Scale Factor=&Multiply_VectorByFloat(Vector=<1 1 1>, Float=1))
   &DynamicMaterialParameters(
    Index 0 Param 1=&MakeFloatFromVector(Vector=$Particles.Scale, Channel=::NewEnumerator0)
    Index 0 Param 2=&RandomRangeFloat(Minimum=0, Maximum=10)
    Index 0 Param 3=&FloatFromCurve(
     FloatCurve=DataInterface() {propertyValues {Curve {keys [{interpMode=RCIM_Cubic, leaveTangent=8.1085, tangentMode=RCTM_Break, time=0.8568, value=1}, {tangentMode=RCTM_User, time=1, value=0}]}}}
     Scale Curve=0.1
    )
    Index 0 Param 4=0.5
    Index 1 Param 1=&FloatFromCurve(FloatCurve=DataInterface() {propertyValues {Curve {keys [
         {interpMode=RCIM_Cubic, leaveTangent=1.1034, tangentMode=RCTM_Break, time=0, value=0}
         {arriveTangent=2.2399, tangentMode=RCTM_User, time=0.5, value=1}
        ]}}})
    Index 1 Param 2=&FloatFromCurve(FloatCurve=DataInterface() {propertyValues {Curve {keys [<at 0.5 0>, {arriveTangent=3.2053, tangentMode=RCTM_User, time=0.9, value=1}]}}})
    Index 1 Param 3=&FloatFromCurve(FloatCurve=DataInterface() {propertyValues {Curve {keys [<at 0 0 RCIM_Cubic>, {arriveTangent=3.4223, interpMode=RCIM_Cubic, leaveTangent=3.4223, tangentMode=RCTM_User, time=1, value=1}]}}})
    Index 1 Param 4=&FloatFromCurve(
     FloatCurve=DataInterface() {propertyValues {Curve {keys [
         {interpMode=RCIM_Cubic, leaveTangent=0.1169, tangentMode=RCTM_Break, time=0, value=0}
         {arriveTangent=2.5708, interpMode=RCIM_Cubic, leaveTangent=2.5708, tangentMode=RCTM_User, time=1, value=1}
        ]}}}
     Scale Curve=&RandomRangeFloat(Minimum=1, Maximum=40)
    )
    Index 2 Param 1=&FloatFromCurve(FloatCurve=DataInterface() {propertyValues {Curve {keys [
         {interpMode=RCIM_Cubic, leaveTangent=-0.0025, tangentMode=RCTM_Break, time=0, value=0.5}
         {arriveTangent=-0.0316, interpMode=RCIM_Cubic, tangentMode=RCTM_Break, time=1, value=-0.2}
        ]}}})
    Index 2 Param 2=&FloatFromCurve(
     FloatCurve=DataInterface() {propertyValues {Curve {keys [
         {interpMode=RCIM_Cubic, tangentMode=RCTM_User, time=0, value=0}
         {arriveTangent=2.0522, interpMode=RCIM_Cubic, leaveTangent=2.0522, tangentMode=RCTM_User, time=1, value=1}
        ]}}}
     Scale Curve=2
    )
    Module.Index0Param0Write=true
    Module.Index0Param1Write=true
    Module.Index0Param2Write=true
    Module.Index0Param3Write=true
    Module.Index1Param0Write=true
    Module.Index1Param1Write=true
    Module.Index1Param2Write=true
    Module.Index1Param3Write=true
    Module.Index2Param0Write=true
    Module.Index2Param1Write=true
    Module.Index2Param2Write=false
    Module.Index2Param3Write=false
    Module.Index3Param0Write=false
    Module.Index3Param1Write=false
    Module.Index3Param2Write=false
    Module.Index3Param3Write=true
    Write Parameter Index 1=true
    Write Parameter Index 2=true
    Write Parameter Index 3=true
   )
   &UpdateMeshOrientation(Rotation Rate=&RandomRangeFloat(Minimum=-1, Maximum=1), Rotation Vector=<0 0 1>)
   &ScaleColor(
    Scale Alpha=&FloatFromCurve(FloatCurve=DataInterface() {propertyValues {Curve {keys [
         {tangentMode=RCTM_User, time=0, value=1}
         {arriveTangent=-0.0405, interpMode=RCIM_Cubic, leaveTangent=-0.0405, tangentMode=RCTM_User, time=0.7009, value=1}
         {arriveTangent=-8.0089, interpMode=RCIM_Cubic, leaveTangent=-8.0089, tangentMode=RCTM_User, time=1, value=0}
        ]}}})
    Scale RGB=&Lerp_Vector(
     A=&VectorFromFloat(Value=&FloatFromCurve(
       FloatCurve=DataInterface() {propertyValues {Curve {keys [<at 0 1>,<at 0.1 1>,<at 0.2 2>,<at 0.3 1>,<at 0.6 1>,<at 0.7 2>,<at 0.8 1>,<at 1 1>]}}}
       Scale Curve=10
      ))
     B=&VectorFromFloat(Value=&RandomRangeFloat(Minimum=0.3, Maximum=5, Evaluation Type=::NewEnumerator1))
     Alpha=&FloatFromCurve(FloatCurve=DataInterface() {propertyValues {Curve {keys [
          {interpMode=RCIM_Cubic, leaveTangent=2.5294, tangentMode=RCTM_Break, time=0, value=0}
          {arriveTangent=-0.2166, interpMode=RCIM_Cubic, tangentMode=RCTM_Break, time=0.4, value=1}
         ]}}})
    )
    ScaleA=true
    ScaleRGB=true
   )
  ]
  rendererClasses [&NiagaraMeshRendererProperties]
  renderers [
   {rendererClass=&NiagaraMeshRendererProperties, rendererIndex=0}
   {rendererClass=&NiagaraMeshRendererProperties, rendererIndex=1}
  ]
 }
 LightningSparks {
  emitterUpdateScript [
   &EmitterState(Life Cycle Mode=::NewEnumerator1, Loop Behavior=::NewEnumerator1, Loop Duration Mode=::NewEnumerator1, UseLoopDelay=false)
   &UpdateValueStateChange(InputFloatParam:NiagaraFloat=$User.BurstIndex, ParamType:ENiagaraValueStateChange=::NewEnumerator1)
   &self:TriggerSpawn(
    Spawn Count:NiagaraInt32=&RandomRangeInt(Minimum=18, Maximum=25)
    Trigger:NiagaraBool=$Emitter.UpdateValueStateChange.Changed
   ) {module=TriggerSpawn}
   &SpawnRate(SpawnRate=50, Use Spawn Probability=false) {enabled=false}
   &SpawnBurst_Instantaneous(
    Spawn Count=&RandomRangeInt(Minimum=15, Maximum=25)
    Use Spawn Probability=false
    UseLoopCountLimit=false
   ) {enabled=false}
  ]
  particleSpawnScript [
   &InitializeParticle(
    AdjustAlpha=false
    AdjustHue=true
    AdjustSaturation=false
    AdjustValue=true
    Color=$System.LightningColor
    Color Mode=::NewEnumerator3
    Hue Shift Range:Vector2f=<-0.01 0.01>
    Lifetime Max=&RandomRangeFloat(Minimum=0.06, Maximum=0.13)
    Lifetime Min=0.05
    Lifetime Mode=::NewEnumerator1
    Mass Max=2
    Mass Mode=::NewEnumerator3
    Position Offset=<0 0 10>
    Position Offset Coordinate Space=::Local
    Sprite Size Mode=::NewEnumerator1
    Uniform Sprite Size Max=5
    Uniform Sprite Size Min=2
    UsePositionOffset=true
    Value Range:Vector2f=<0.8 5>
    Write Material Random=true
   )
   &AddVelocity(
    Rotation Coordinate Space=::Simulation
    Velocity=&RandomRangeVector(Minimum=<-1 0 -0.01>, Maximum=<-1 0 -0.01>)
    Velocity Speed Scale=&RandomRangeFloat(
     Minimum=&Multiply_Float(A=&RandomRangeFloat(Minimum=-300, Maximum=-400), B=$System.FrontEndScale)
     Maximum=&Multiply_Float(A=&RandomRangeFloat(Minimum=300, Maximum=400), B=$System.FrontEndScale)
    )
   ) {enabled=false, module=AddVelocity001}
   &AddVelocity(
    Cone Angle=110
    Cone Axis=<0 1 0>
    Distribution Along Cone Axis=0.2
    Module.UseConeFalloff=false
    Module.UseVelDistribution=true
    Override Seed=false
    Rotation Axis:Vector3f=<0 0 1>
    Rotation Mode=::NewEnumerator1
    Velocity Mode=::NewEnumerator2
    Velocity Speed=&RandomRangeFloat(Minimum=300, Maximum=300)
   )
  ]
  particleUpdateScript [
   &ParticleState()
   &GravityForce()
   &Drag(Drag=&RandomRangeFloat(Minimum=0.8, Maximum=1.2), Use Linear Drag=true, Use Rotational Drag=false)
   &ScaleColor(
    Scale Alpha=&FloatFromCurve(FloatCurve=DataInterface() {propertyValues {Curve {keys [
         {arriveTangent=9.0725, interpMode=RCIM_Cubic, leaveTangent=9.0725, tangentMode=RCTM_User, time=0, value=0}
         {arriveTangent=-2.5249, interpMode=RCIM_Cubic, leaveTangent=-2.5249, tangentMode=RCTM_User, time=0.1408, value=0.9847}
         {arriveTangent=-0.5654, interpMode=RCIM_Cubic, leaveTangent=-0.5654, time=0.5, value=0.35}
         {arriveTangent=-1.25, interpMode=RCIM_Cubic, leaveTangent=-1.25, time=0.6, value=0.725}
         {arriveTangent=-3.375, interpMode=RCIM_Cubic, leaveTangent=-0.1412, tangentMode=RCTM_Break, time=0.7, value=0.1}
         {arriveTangent=-0.7119, interpMode=RCIM_Cubic, leaveTangent=1.625, tangentMode=RCTM_Break, time=0.8, value=0.05}
         {arriveTangent=-0.25, interpMode=RCIM_Cubic, leaveTangent=-0.25, time=0.9, value=0.425}
         {arriveTangent=-0.0405, interpMode=RCIM_Cubic, leaveTangent=-0.0405, tangentMode=RCTM_User, time=1, value=0}
        ]}}})
    Scale RGB=&VectorFromFloat(Value=&FloatFromCurve(
      FloatCurve=DataInterface() {propertyValues {Curve {keys [<at 0 1>,<at 0.2 1>,<at 0.3 5>,<at 0.4 1>,<at 0.7 1>,<at 0.8 5>,<at 0.9 1>,<at 1 1>]}}}
      Scale Curve=2
     ))
    ScaleA=true
    ScaleRGB=true
   ) {module=ScaleColor001}
   &SolveForcesAndVelocity(Clamp Velocity=false, Limit Acceleration=false)
   &ScaleSpriteSizeBySpeed(Max Scale Factor=<0.5 7>, Sample Scale Factor By Curve=true, Velocity Threshold=500)
  ]
  rendererClasses [&NiagaraSpriteRendererProperties]
  renderers [{rendererClass=&NiagaraSpriteRendererProperties}]
 }
 StaffLightning {
  emitterUpdateScript [
   &EmitterState(
    Inactive Response=::NewEnumerator1
    Life Cycle Mode=::NewEnumerator1
    Loop Behavior=::NewEnumerator1
    Loop Duration Mode=::NewEnumerator1
    UseLoopDelay=false
   )
   &SpawnRate(Spawn Probability=0.5, SpawnRate=15, Use Spawn Probability=false)
   &SpawnRate(Spawn Probability=0.5, SpawnGroup=1, SpawnRate=15, Use Spawn Probability=false) {module=SpawnRate001}
  ]
  particleSpawnScript [
   /SetVariables(Particles.VisibilityTag=&SpawnGroup()) {isSetParameters=true}
   &InitializeParticle(
    Color=&Multiply_LinearColorByFloat(Linear Color=$System.LightningColor, Float=0.4)
    Lifetime Max=0.35
    Lifetime Min=0.175
    Lifetime Mode=::NewEnumerator1
    Mesh Scale Max=<2.1 2.1 2.5>
    Mesh Scale Min=<1 1 2.5>
    Mesh Scale Mode=::NewEnumerator2
    UsePositionOffset=false
    Write Material Random=true
   )
   &ShapeLocation(
    Box Midpoint=&MaskVectorBySpawnGroup(Masked Vector=<0.5 0.5 1.75>, Passthrough Vector=<0.5 0.5 0.75>)
    Box Size=&MaskVectorBySpawnGroup(Masked Vector=<5 5 40>, Passthrough Vector=<2 2 40>)
    MaskSpawnGroup=false
    Offset=<0 0 -2.5>
    Offset Mode=::NewEnumerator0
    Override Seed=false
    Shape Primitive=::NewEnumerator2
    Surface Only=false
    UseOwnerScale=true
   )
   &InitialMeshOrientation(
    Mesh Orientation Mode=::NewEnumerator2
    Rotation=&RandomRangeVector(Minimum=<0 0 -1>, Maximum=<0 0 1>)
    Use Rotation Vector=true
   )
   &AddVelocity(
    Velocity=<0 0 1>
    Velocity Speed Scale=&RandomRangeFloat(Minimum=5, Maximum=&MaskFloatBySpawnGroup(Masked Float=5, Passthrough Float=15))
   )
  ]
  particleUpdateScript [
   &SolveForcesAndVelocity(Clamp Velocity=false, Limit Acceleration=false)
   &ParticleState()
   &ScaleMeshSize(Scale Factor=&Multiply_VectorByFloat(Vector=<1 1 1>, Float=0.2))
   &DynamicMaterialParameters(
    Index 0 Param 1=&MakeFloatFromVector(Vector=$Particles.Scale, Channel=::NewEnumerator0)
    Index 0 Param 2=&RandomRangeFloat(Minimum=0, Maximum=10)
    Index 0 Param 3=&FloatFromCurve(
     FloatCurve=DataInterface() {propertyValues {Curve {keys [{interpMode=RCIM_Cubic, leaveTangent=8.1085, tangentMode=RCTM_Break, time=0.8568, value=1}, {tangentMode=RCTM_User, time=1, value=0}]}}}
     Scale Curve=0.07
    )
    Index 0 Param 4=1
    Index 1 Param 1=&FloatFromCurve(FloatCurve=DataInterface() {propertyValues {Curve {keys [
         {interpMode=RCIM_Cubic, leaveTangent=1.1034, tangentMode=RCTM_Break, time=0, value=0}
         {arriveTangent=2.2399, tangentMode=RCTM_User, time=0.5, value=1}
        ]}}})
    Index 1 Param 2=&FloatFromCurve(FloatCurve=DataInterface() {propertyValues {Curve {keys [<at 0.5 0>, {arriveTangent=3.2053, tangentMode=RCTM_User, time=0.9, value=1}]}}})
    Index 1 Param 3=&FloatFromCurve(FloatCurve=DataInterface() {propertyValues {Curve {keys [<at 0 0 RCIM_Cubic>, {arriveTangent=3.4223, interpMode=RCIM_Cubic, leaveTangent=3.4223, tangentMode=RCTM_User, time=1, value=1}]}}})
    Index 1 Param 4=&FloatFromCurve(
     FloatCurve=DataInterface() {propertyValues {Curve {keys [
         {interpMode=RCIM_Cubic, leaveTangent=0.1169, tangentMode=RCTM_Break, time=0, value=0}
         {arriveTangent=2.5708, interpMode=RCIM_Cubic, leaveTangent=2.5708, tangentMode=RCTM_User, time=1, value=1}
        ]}}}
     Scale Curve=&RandomRangeFloat(Minimum=1, Maximum=40)
    )
    Index 2 Param 1=&FloatFromCurve(FloatCurve=DataInterface() {propertyValues {Curve {keys [
         {interpMode=RCIM_Cubic, leaveTangent=-0.0025, tangentMode=RCTM_Break, time=0, value=0.5}
         {arriveTangent=-0.0316, interpMode=RCIM_Cubic, tangentMode=RCTM_Break, time=1, value=-0.2}
        ]}}})
    Index 2 Param 2=&FloatFromCurve(
     FloatCurve=DataInterface() {propertyValues {Curve {keys [
         {interpMode=RCIM_Cubic, tangentMode=RCTM_User, time=0, value=0}
         {arriveTangent=2.0522, interpMode=RCIM_Cubic, leaveTangent=2.0522, tangentMode=RCTM_User, time=1, value=1}
        ]}}}
     Scale Curve=2
    )
    Module.Index0Param0Write=true
    Module.Index0Param1Write=true
    Module.Index0Param2Write=true
    Module.Index0Param3Write=true
    Module.Index1Param0Write=true
    Module.Index1Param1Write=true
    Module.Index1Param2Write=true
    Module.Index1Param3Write=true
    Module.Index2Param0Write=true
    Module.Index2Param1Write=true
    Module.Index2Param2Write=false
    Module.Index2Param3Write=false
    Module.Index3Param0Write=false
    Module.Index3Param1Write=false
    Module.Index3Param2Write=false
    Module.Index3Param3Write=true
    Write Parameter Index 1=true
    Write Parameter Index 2=true
    Write Parameter Index 3=true
   )
   &UpdateMeshOrientation(Rotation Rate=&RandomRangeFloat(Minimum=-1, Maximum=1), Rotation Vector=<0 0 1>)
   &ScaleColor(
    Scale Alpha=&FloatFromCurve(FloatCurve=DataInterface() {propertyValues {Curve {keys [
         {tangentMode=RCTM_User, time=0, value=1}
         {arriveTangent=-0.0405, interpMode=RCIM_Cubic, leaveTangent=-0.0405, tangentMode=RCTM_User, time=0.7009, value=1}
         {arriveTangent=-8.0089, interpMode=RCIM_Cubic, leaveTangent=-8.0089, tangentMode=RCTM_User, time=1, value=0}
        ]}}})
    Scale RGB=&Lerp_Vector(
     A=&VectorFromFloat(Value=&FloatFromCurve(
       FloatCurve=DataInterface() {propertyValues {Curve {keys [<at 0 1>,<at 0.1 1>,<at 0.2 2>,<at 0.3 1>,<at 0.6 1>,<at 0.7 2>,<at 0.8 1>,<at 1 1>]}}}
       Scale Curve=10
      ))
     B=&VectorFromFloat(Value=&RandomRangeFloat(Minimum=0.3, Maximum=5, Evaluation Type=::NewEnumerator1))
     Alpha=&FloatFromCurve(FloatCurve=DataInterface() {propertyValues {Curve {keys [
          {interpMode=RCIM_Cubic, leaveTangent=2.5294, tangentMode=RCTM_Break, time=0, value=0}
          {arriveTangent=-0.2166, interpMode=RCIM_Cubic, tangentMode=RCTM_Break, time=0.4, value=1}
         ]}}})
    )
    ScaleA=true
    ScaleRGB=true
   )
  ]
  rendererClasses [&NiagaraMeshRendererProperties]
  renderers [
   {rendererClass=&NiagaraMeshRendererProperties, rendererIndex=0}
   {rendererClass=&NiagaraMeshRendererProperties, rendererIndex=1}
   {rendererClass=&NiagaraMeshRendererProperties, rendererIndex=2}
   {rendererClass=&NiagaraMeshRendererProperties, rendererIndex=3}
  ]
 }
}
summary {
 systemName=NS_Emote_RainbowRaid_Lightning
 emitters [
  {emitterName=StaffLightning, rendererClasses [&NiagaraMeshRendererProperties]}
  {emitterName=Glow, rendererClasses []}
  {emitterName=Lightning, rendererClasses [&NiagaraMeshRendererProperties]}
  {emitterName=LightningSparks, rendererClasses [&NiagaraSpriteRendererProperties]}
  {emitterName=Flash_Flare, rendererClasses [&NiagaraSpriteRendererProperties]}
  {emitterName=Light, rendererClasses []}
 ]
 userVariables [
  User.bIsFrontEnd {type=NiagaraBool}
  User.bIsFrontEndPreview {type=NiagaraBool}
  User.BurstIndex {type=NiagaraFloat, defaultValue=0, description=""}
 ]
}
