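# Niagara lightning spawn effect for the RainbowRaid emote, built from three self-timed emitters.
# Lightning spawns randomized meshes with animated material parameters and flickering color, using System.LightningColor and System.FadeOut.
# Burst pairs a flash sprite with curl-noise sparks and particle lights, while Corona adds one expanding, fading glow sprite.
# Spawn-group masks give particle types within an emitter different behavior, and time-of-day scalars adjust sprite alpha.
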NiagaraSystem /BRCosmetics/Animation/Game/MainPlayer/Emotes/RainbowRaid/FX/NS_Emote_RainbowRaid_LightningSpawn
dependencies {
 usedDataInterfaces [&NiagaraDataInterfaceCurve, &NiagaraDataInterfaceColorCurve]
 usedDynamicInputs [
  &MakeCustomVectorFromBool
  &SetBoolByFloatComparison
  &UniformRangedVector
 ]
 usedModules [
  &self:SystemSpawnScript.SystemScriptSource.SystemScriptGraph.NiagaraNodeAssignment_0.SetVariables_#1
  &SystemState
 ]
}
emitters {
 Burst {
  emitterUpdateScript [
   &EmitterState(Life Cycle Mode=::NewEnumerator1, Loop Behavior=::NewEnumerator1, Loop Duration=0.6, UseLoopDelay=false)
   &SpawnRate(
    SpawnGroup=1
    SpawnRate=&FloatFromCurve(
     FloatCurve=DataInterface() {propertyValues {Curve {keys [<at 0 1 RCIM_Constant>,<at 0.2 -0.05 RCIM_Constant>,<at 1 0>]}}}
     CurveIndex=$Emitter.NormalizedLoopAge
     Scale Curve=75
    )
    Use Spawn Probability=false
   )
   &SpawnBurst_Instantaneous(Spawn Count=1, Use Spawn Probability=false)
  ]
  particleSpawnScript [
   /SetVariables(Particles.VisibilityTag=&SpawnGroup()) {isSetParameters=true}
   &InitializeParticle(
    Color=&MaskLinearColorBySpawnGroup(
     Masked Linear Color=<rgba 0.5984 1.9636 5 1>
     Passthrough Linear Color=&Multiply_LinearColorByFloat(Linear Color=$System.LightningColor, Float=&RandomRangeFloat(Minimum=8, Maximum=10))
    )
    Lifetime=&MaskFloatBySpawnGroup(Masked Float=0.3, Passthrough Float=&RandomRangeFloat(Minimum=0.1, Maximum=1))
    Mass Mode=::NewEnumerator1
    Mesh Scale Mode=::NewEnumerator3
    Position=&Add_Vector(A=/Niagara/DynamicInputs/Helpers/SimulationPosition(), B=<0 0 0>)
    Position Mode=::NewEnumerator1
    Sprite Size=&MaskVector2dBySpawnGroup(
     Masked Vector2D=<120 120>
     Passthrough Vector2D=&RandomRangeVector2D(Minimum=<4 6>, Maximum=<5 7>)
     Spawn Group Mask=0
    )
    Sprite Size Mode=::NewEnumerator4
    UsePositionOffset=false
    Write Material Random=true
   ) {module=InitializeParticle001}
   &CameraOffset(Camera Offset Amount=&MaskFloatBySpawnGroup(Masked Float=10, Passthrough Float=0))
   &SpriteFacingAndAlignment(Sprite Alignment=<0 0 1>, WriteAlignment=true, WriteFacing=false)
   &AddVelocity(
    Cone Angle=60
    Cone Axis=<0 0 2>
    Distribution Along Cone Axis:NiagaraFloat=&MaskFloatBySpawnGroup(Masked Float=0, Passthrough Float=0.2)
    Module.UseConeFalloff=false
    Module.UseVelDistribution=true
    Override Seed=false
    Velocity Mode=::NewEnumerator2
    Velocity Speed=&MaskFloatBySpawnGroup(Masked Float=0, Passthrough Float=&RandomRangeFloat(Minimum=150, Maximum=350))
   )
  ]
  particleUpdateScript [
   &CurlNoiseForce(
    Bias Noise=false
    Noise Strength=&MaskFloatBySpawnGroup(
     Masked Float=0
     Passthrough Float=&FloatFromCurve(
      FloatCurve=DataInterface() {propertyValues {Curve {keys [{interpMode=RCIM_Cubic, tangentMode=RCTM_User, time=0, value=0}, {interpMode=RCIM_Cubic, tangentMode=RCTM_User, time=1, value=1}]}}}
      Scale Curve=1000
     )
    )
    Pan Noise=true
    Pan Noise Field=&MaskVectorBySpawnGroup(Masked Vector=<0 0 0>, Passthrough Vector=<0 0 1>)
   )
   &Drag(
    Drag=&MaskFloatBySpawnGroup(
     Masked Float=0
     Passthrough Float=&FloatFromCurve(
      FloatCurve=DataInterface() {propertyValues {Curve {keys [{interpMode=RCIM_Cubic, tangentMode=RCTM_User, time=0, value=1}, {interpMode=RCIM_Cubic, tangentMode=RCTM_User, time=1, value=0}]}}}
      Scale Curve=&RandomRangeFloat(Minimum=1, Maximum=7)
     )
    )
    Rotational Drag:NiagaraFloat=&MaskFloatBySpawnGroup(Masked Float=0, Passthrough Float=1)
    Use Linear Drag=true
    Use Rotational Drag=true
   ) {module=Drag001}
   &GravityForce(Gravity=&MaskVector(
     Mask Source=0
     Mask Parameter=?
     Mask=0
     Masked Vector=<0 0 0>
     Passthrough Vector=&RandomRangeVector(Minimum=<0 0 -100>, Maximum=<0 0 -200>)
    ))
   &SolveForcesAndVelocity(Clamp Velocity=false, Limit Acceleration=false)
   &ParticleState()
   &ScaleSpriteSize(
    Scale Factor=&MaskVector2dBySpawnGroup(
     Masked Vector2D=&Multiply_Vector2D_ByFloat(
      Float=&FloatFromCurve(FloatCurve=DataInterface() {propertyValues {Curve {keys [
           {interpMode=RCIM_Cubic, leaveTangentWeight=0.504, tangentMode=RCTM_User, time=0, value=0.5}
           {arriveTangentWeight=0.1, interpMode=RCIM_Cubic, leaveTangentWeight=0.2333, tangentMode=RCTM_User, time=0.0873, value=1.5558}
           {arriveTangent=-0.6369, interpMode=RCIM_Cubic, leaveTangent=-0.6369, tangentMode=RCTM_User, time=0.1744, value=1.555}
           {arriveTangent=-2.2429, interpMode=RCIM_Cubic, leaveTangent=-2.2429, time=0.3044, value=0.2397}
           {arriveTangent=4.7846, interpMode=RCIM_Cubic, leaveTangent=4.7846, time=0.3726, value=1.1104}
           {arriveTangent=-3.8005, interpMode=RCIM_Cubic, leaveTangent=-3.8005, time=0.4864, value=1.1104}
           {arriveTangent=-0.4501, interpMode=RCIM_Cubic, leaveTangent=-0.4501, time=0.5676, value=0.3694}
           {arriveTangent=1.4585, interpMode=RCIM_Cubic, leaveTangent=1.4585, time=0.7334, value=0.9992}
           {tangentMode=RCTM_User, time=1, value=1}
          ]}}})
      Vector2D=<1 1>
     )
     Passthrough Vector2D=&Multiply_Vector2D_ByFloat(
      Float=&FloatFromCurve(FloatCurve=DataInterface() {propertyValues {Curve {keys [{interpMode=RCIM_Cubic, tangentMode=RCTM_User, time=0, value=1}, {interpMode=RCIM_Cubic, tangentMode=RCTM_User, time=1, value=0}]}}})
      Vector2D=<1 1>
     )
     Spawn Group Mask=0
    )
    Scale Sprite Size Mode=::NewEnumerator2
   )
   &ScaleSpriteSizeBySpeed(
    Max Scale Factor=&MaskVector2dBySpawnGroup(Masked Vector2D=<1 1>, Passthrough Vector2D=<0.75 6>, Spawn Group Mask=0)
    Min Scale Factor=&MaskVector2dBySpawnGroup(Masked Vector2D=<1 1>, Passthrough Vector2D=<0.75 0.5>, Spawn Group Mask=0)
    Sample Scale Factor By Curve=true
    Scale Factor Curve=DataInterface() {propertyValues {Curve {keys [
        {arriveTangent=2.865, interpMode=RCIM_Cubic, time=0.1247, value=-0.0006}
        {arriveTangent=4.1522, interpMode=RCIM_Cubic, leaveTangent=4.1522, tangentMode=RCTM_User, time=0.3948, value=0.4893}
        <at 0.9958 0.9974 RCIM_Cubic>
       ]}}}
    Velocity Threshold=600
   )
   &FN_Sprite_Attributes(
    Create SpriteSize Attribute=true
    SpriteSize=&Multiply_Vector2D(
     A=<75 40>
     B=&Vector2DFromFloat(Value=&FloatFromCurve(FloatCurve=DataInterface() {propertyValues {Curve {keys [<at 0 0 RCIM_Cubic>,<at 0.5 1 RCIM_Cubic>]}}}))
    )
    Write to Module Namespace=true
   ) {module=FN_Sprite}
   &ScaleColor(
    Linear Color Curve=DataInterface() {propertyValues {
      AlphaCurve {keys [<at 0 0.4>,<at 1 0>]}
      BlueCurve {keys [<at 0 2>,<at 0.5363 2>,<at 0.989 1>]}
      GreenCurve {keys [<at 0 2>,<at 0.5363 2>,<at 0.989 1>]}
      RedCurve {keys [<at 0 2>,<at 0.5363 2>,<at 0.989 1>]}
     }}
    Scale Mode=::NewEnumerator2
   )
   &ScaleColor(
    Scale Alpha=&MaskFloatBySpawnGroup(
     Masked Float=1
     Passthrough Float=&FloatFromCurve(FloatCurve=DataInterface() {propertyValues {Curve {keys [
          {arriveTangent=9.0725, interpMode=RCIM_Cubic, leaveTangent=9.0725, tangentMode=RCTM_User, time=0, value=0}
          {arriveTangent=-2.5249, interpMode=RCIM_Cubic, leaveTangent=-2.5249, tangentMode=RCTM_User, time=0.1, value=1.05}
          {arriveTangent=-0.6875, interpMode=RCIM_Cubic, leaveTangent=-0.6875, time=0.4, value=0.45}
          {arriveTangent=-1, interpMode=RCIM_Cubic, leaveTangent=-1, time=0.5, value=0.775}
          {arriveTangent=-1.25, interpMode=RCIM_Cubic, leaveTangent=-1.25, time=0.6, value=0.25}
          {arriveTangent=-0.375, interpMode=RCIM_Cubic, leaveTangent=-0.375, time=0.7, value=0.525}
          {arriveTangent=-1.75, interpMode=RCIM_Cubic, leaveTangent=-1.75, time=0.8, value=0.175}
          {arriveTangent=-0.0405, interpMode=RCIM_Cubic, leaveTangent=-0.0405, tangentMode=RCTM_User, time=1, value=0}
         ]}}})
    )
    Scale RGB=&MaskVectorBySpawnGroup(
     Masked Vector=<1 1 1>
     Passthrough Vector=&VectorFromFloat(Value=&FloatFromCurve(FloatCurve=DataInterface() {propertyValues {Curve {keys [<at 0 1>,<at 0.1 5>,<at 0.2 1>,<at 0.6 1>,<at 0.7 3>,<at 0.8 1>,<at 1 1>]}}}))
    )
    ScaleA=true
    ScaleRGB=true
   ) {module=ScaleColor002}
   &FN_Light_Attributes(
    Color=$System.LightningColor
    Create Color Attribute=true
    Create Exponent Attribute=true
    Create Position Attribute=true
    Create Radius Attribute=true
    Enable Particle Light Enable=true
    Exponent=8
    Light Enabled=true
    Position=&Add_Vector(
     A=Linked() {linkedVariable {name=Particles.Position, type=NiagaraPosition}}
     B=/Niagara/DynamicInputs/Transforms/TransformPosition(Position=<0 0 0>, Source Space=::Local, Destination Space=::Simulation)
    )
    Radius=200
    Scale Color=&FloatFromCurve(
     FloatCurve=DataInterface() {propertyValues {Curve {keys [
         {arriveTangent=9.0725, interpMode=RCIM_Cubic, leaveTangent=9.0725, tangentMode=RCTM_User, time=0, value=0}
         {arriveTangent=-2.5249, interpMode=RCIM_Cubic, leaveTangent=-2.5249, tangentMode=RCTM_User, time=0.1408, value=0.9847}
         {arriveTangent=-0.0405, interpMode=RCIM_Cubic, leaveTangent=-0.0405, tangentMode=RCTM_User, time=1, value=0}
        ]}}}
     Scale Curve=10
    )
   )
   &ScaleColor(
    Scale Alpha=&FNTimeOfDayScalar(Night Scale=0.4, Morning Scale=0.7, Evening Scale=0.5)
    ScaleA=true
    ScaleRGB=false
   ) {module=ScaleColor001}
  ]
  rendererClasses [&NiagaraLightRendererProperties, &NiagaraSpriteRendererProperties]
  renderers [
   {rendererClass=&NiagaraLightRendererProperties, rendererIndex=0}
   {rendererClass=&NiagaraSpriteRendererProperties, rendererIndex=1}
   {rendererClass=&NiagaraSpriteRendererProperties, rendererIndex=2}
   {rendererClass=&NiagaraSpriteRendererProperties, rendererIndex=3}
   {rendererClass=&NiagaraSpriteRendererProperties, rendererIndex=4}
  ]
 }
 Corona {
  emitterSpawnScript [/SetVariables() {isSetParameters=true}]
  emitterUpdateScript [
   &EmitterState(Life Cycle Mode=::NewEnumerator1, Loop Behavior=::NewEnumerator1, Loop Duration=1, UseLoopDelay=false)
   &SpawnBurst_Instantaneous(Spawn Count=1, Spawn Time=0.05, Use Spawn Probability=false)
  ]
  particleSpawnScript [
   &InitializeParticle(
    Lifetime=0.25
    Sprite Rotation Mode=::NewEnumerator2
    Sprite Size Mode=::NewEnumerator3
    Uniform Sprite Size=100
    UsePositionOffset=false
    Write Material Random=true
   )
   &DynamicMaterialParameters(
    Index 0 Param 1=&RandomRangeFloat(Minimum=0.4, Maximum=0.6)
    Index 0 Param 2=1
    Index 0 Param 3=1
    Index 0 Param 4=1
    Module.Index0Float0Write=true
    Module.Index0Float1Write=true
    Module.Index0Float2Write=true
    Module.Index0Float3Write=true
   )
  ]
  particleUpdateScript [
   &ParticleState()
   &SolveForcesAndVelocity(Clamp Velocity=false, Limit Acceleration=false)
   &ScaleColor(
    Scale Alpha=&Multiply_Float(
     A=&FNTimeOfDayScalar(Night Scale=1, Morning Scale=1, Evening Scale=1)
     B=&FloatFromCurve(FloatCurve=DataInterface() {propertyValues {Curve {keys [
          {arriveTangent=9.6062, interpMode=RCIM_Cubic, leaveTangent=9.6062, tangentMode=RCTM_User, time=0, value=0}
          {arriveTangent=-2.0684, interpMode=RCIM_Cubic, leaveTangent=-2.0684, tangentMode=RCTM_User, time=0.2, value=1}
          {interpMode=RCIM_Cubic, tangentMode=RCTM_User, time=1, value=0}
         ]}}})
    )
    Scale RGB=&Multiply_VectorByFloat(Vector=<1 1 1>, Float=1)
    ScaleA=true
    ScaleRGB=true
   )
   &ScaleSpriteSize(Scale Factor=&Vector2DFromFloat(Value=&FloatFromCurve(FloatCurve=DataInterface() {propertyValues {Curve {keys [
          {arriveTangent=2.1722, interpMode=RCIM_Cubic, leaveTangent=2.1722, tangentMode=RCTM_User, time=0, value=0}
          {arriveTangent=0.0224, interpMode=RCIM_Cubic, leaveTangent=0.0224, tangentMode=RCTM_User, time=1, value=1}
         ]}}})))
   &CameraOffset(Camera Offset Amount=-30)
   &DynamicMaterialParameters(
    Index 0 Param 1=0.75
    Index 0 Param 2=2
    Index 0 Param 3=1
    Module.Param0WriteEnabled=true
    Module.Param1WriteEnabled=true
    Module.Param2WriteEnabled=true
    Module.Param3WriteEnabled=true
   ) {module=DynamicMaterialParameters001}
  ]
  rendererClasses [&NiagaraSpriteRendererProperties]
  renderers [{rendererClass=&NiagaraSpriteRendererProperties}]
 }
 Lightning {
  emitterUpdateScript [
   &EmitterState(Inactive Response=::NewEnumerator1, Life Cycle Mode=::NewEnumerator1, Loop Behavior=::NewEnumerator1, Loop Duration=0.5, UseLoopDelay=false)
   &SpawnRate(SpawnRate=&Multiply_Float(A=25, B=$System.FadeOut), Use Spawn Probability=false)
   &SpawnBurst_Instantaneous(Use Spawn Probability=false, UseLoopCountLimit=false)
   &SpawnRate(
    SpawnGroup=1
    SpawnRate=&FloatFromCurve(
     FloatCurve=DataInterface() {propertyValues {Curve {
        defaultValue=1
        keys [
         {interpMode=RCIM_Cubic, tangentMode=RCTM_User, time=0, value=1}
         {arriveTangent=-5.3323, interpMode=RCIM_Cubic, leaveTangent=-5.3323, tangentMode=RCTM_User, time=0.4842, value=0}
        ]
       }}}
     CurveIndex=$Emitter.NormalizedLoopAge
     Scale Curve=30
    )
    Use Spawn Probability=false
   ) {module=SpawnRate001}
   &SpawnBurst_Instantaneous(
    Spawn Count=&RandomRangeInt(Minimum=5, Maximum=7)
    Spawn Group=1
    Use Spawn Probability=false
    UseLoopCountLimit=false
   ) {module=SpawnBurst_Instantaneous001}
  ]
  particleSpawnScript [
   /SetVariables(Particles.VisibilityTag=&SpawnGroup()) {isSetParameters=true}
   &InitializeParticle(
    Color=&Multiply_LinearColorByFloat(Linear Color=$System.LightningColor, Float=0.4)
    Lifetime Max=0.45
    Lifetime Min:NiagaraFloat=&MaskFloatBySpawnGroup(
     Masked Float=&RandomRangeFloat(Minimum=0.15, Maximum=0.1)
     Passthrough Float=&RandomRangeFloat(Minimum=0.15, Maximum=0.1)
    )
    Lifetime Mode=::NewEnumerator1
    Mesh Scale Max=&MaskVectorBySpawnGroup(Masked Vector=<1.3 1.3 2.5>, Passthrough Vector=<2 2 3>)
    Mesh Scale Min=&MaskVectorBySpawnGroup(Masked Vector=<1 1 2.5>, Passthrough Vector=<1 1 2>)
    Mesh Scale Mode=::NewEnumerator2
    UsePositionOffset=false
    Write Material Random=true
   )
   &ShapeLocation(
    MaskSpawnGroup=false
    Override Seed=false
    Sphere Radius=&MaskFloatBySpawnGroup(Masked Float=2, Passthrough Float=1)
    UseOwnerScale=true
   )
   &InitialMeshOrientation(
    Mesh Axis To Orient=&MaskVectorBySpawnGroup(Masked Vector=<1 0 0>, Passthrough Vector=<0 0 0>)
    Mesh Orientation Mode=::NewEnumerator2
    Orientation Vector=&MaskVectorBySpawnGroup(Masked Vector=<1 0 0>, Passthrough Vector=<0 0 0>)
    Rotation=&MaskVectorBySpawnGroup(
     Masked Vector=&RandomRangeVector(Minimum=<0 0 -1>, Maximum=<0 0 1>)
     Passthrough Vector=&RandomRangeVector(Minimum=<-1 -1 -0.5>, Maximum=<1 1 0.5>)
    )
    Use Rotation Vector=true
   )
   &AddVelocity(
    Velocity=<0 0 1>
    Velocity Speed Scale=&MaskFloatBySpawnGroup(Masked Float=&RandomRangeFloat(Minimum=5, Maximum=5), Passthrough Float=1)
   )
   &CameraOffset(Camera Offset Amount=&MaskFloatBySpawnGroup(Masked Float=0, Passthrough Float=2))
  ]
  particleUpdateScript [
   &SolveForcesAndVelocity(Clamp Velocity=false, Limit Acceleration=false)
   &ParticleState()
   &ScaleMeshSize(Scale Factor=&MaskVectorBySpawnGroup(
     Masked Vector=<1 1 1>
     Passthrough Vector=&MakeVector(
      X=1
      Y=2
      Z=&FloatFromCurve(FloatCurve=DataInterface() {propertyValues {Curve {keys [
           {interpMode=RCIM_Cubic, leaveTangent=0.0966, tangentMode=RCTM_Break, time=0, value=1}
           {interpMode=RCIM_Cubic, leaveTangent=-0.1745, time=0.7665, value=1.07}
          ]}}})
     )
    )) {module=ScaleMeshSize001}
   &ScaleMeshSize(Scale Factor=&MaskVectorBySpawnGroup(
     Masked Vector=&Multiply_VectorByFloat(Vector=<1 1 1>, Float=&RandomRangeFloat(Minimum=0.45, Maximum=0.5))
     Passthrough Vector=&VectorFromFloat(Value=&RandomRangeFloat(Minimum=0.1, Maximum=0.15))
    ))
   &DynamicMaterialParameters(
    Index 0 Param 1=&MakeFloatFromVector(Vector=$Particles.Scale, Channel=::NewEnumerator0)
    Index 0 Param 2=&RandomRangeFloat(Minimum=0, Maximum=10)
    Index 0 Param 3=&FloatFromCurve(
     FloatCurve=DataInterface() {propertyValues {Curve {keys [{interpMode=RCIM_Cubic, leaveTangent=8.1085, tangentMode=RCTM_Break, time=0.8568, value=1}, {tangentMode=RCTM_User, time=1, value=0}]}}}
     Scale Curve=0.07
    )
    Index 0 Param 4=0.8
    Index 1 Param 1=&FloatFromCurve(FloatCurve=DataInterface() {propertyValues {Curve {keys [
         {interpMode=RCIM_Cubic, leaveTangent=1.1034, tangentMode=RCTM_Break, time=0, value=0}
         {arriveTangent=2.2399, tangentMode=RCTM_User, time=0.5, value=1}
        ]}}})
    Index 1 Param 2=&MaskFloatBySpawnGroup(
     Masked Float=&FloatFromCurve(FloatCurve=DataInterface() {propertyValues {Curve {keys [
          {interpMode=RCIM_Cubic, tangentMode=RCTM_User, time=0, value=0}
          {arriveTangent=2.7735, interpMode=RCIM_Cubic, leaveTangent=2.7735, tangentMode=RCTM_User, time=1, value=1}
         ]}}})
     Passthrough Float=&FloatFromCurve(FloatCurve=DataInterface() {propertyValues {Curve {keys [<at 0.5 0>, {arriveTangent=3.2053, tangentMode=RCTM_User, time=0.9, value=1}]}}})
    )
    Index 1 Param 3=&FloatFromCurve(FloatCurve=DataInterface() {propertyValues {Curve {keys [<at 0 0 RCIM_Cubic>, {arriveTangent=3.4223, interpMode=RCIM_Cubic, leaveTangent=3.4223, tangentMode=RCTM_User, time=1, value=1}]}}})
    Index 1 Param 4=&FloatFromCurve(
     FloatCurve=DataInterface() {propertyValues {Curve {keys [
         {interpMode=RCIM_Cubic, leaveTangent=0.1169, tangentMode=RCTM_Break, time=0, value=0}
         {arriveTangent=2.5708, interpMode=RCIM_Cubic, leaveTangent=2.5708, tangentMode=RCTM_User, time=1, value=1}
        ]}}}
     Scale Curve=&RandomRangeFloat(Minimum=1, Maximum=40)
    )
    Index 2 Param 1=&FloatFromCurve(FloatCurve=DataInterface() {propertyValues {Curve {keys [
         {interpMode=RCIM_Cubic, leaveTangent=-0.0025, tangentMode=RCTM_Break, time=0, value=0.5}
         {arriveTangent=-0.0316, interpMode=RCIM_Cubic, tangentMode=RCTM_Break, time=1, value=-0.2}
        ]}}})
    Index 2 Param 2=&FloatFromCurve(
     FloatCurve=DataInterface() {propertyValues {Curve {keys [
         {interpMode=RCIM_Cubic, tangentMode=RCTM_User, time=0, value=0}
         {arriveTangent=2.0522, interpMode=RCIM_Cubic, leaveTangent=2.0522, tangentMode=RCTM_User, time=1, value=1}
        ]}}}
     Scale Curve=2
    )
    Module.Index0Param0Write=true
    Module.Index0Param1Write=true
    Module.Index0Param2Write=true
    Module.Index0Param3Write=true
    Module.Index1Param0Write=true
    Module.Index1Param1Write=true
    Module.Index1Param2Write=true
    Module.Index1Param3Write=true
    Module.Index2Param0Write=true
    Module.Index2Param1Write=true
    Module.Index2Param2Write=false
    Module.Index2Param3Write=false
    Module.Index3Param0Write=false
    Module.Index3Param1Write=false
    Module.Index3Param2Write=false
    Module.Index3Param3Write=true
    Write Parameter Index 1=true
    Write Parameter Index 2=true
    Write Parameter Index 3=true
   )
   &UpdateMeshOrientation(
    Rotation Rate=&MaskFloatBySpawnGroup(Masked Float=&RandomRangeFloat(Minimum=-1, Maximum=1), Passthrough Float=0)
    Rotation Vector=<0 0 1>
   )
   &ScaleColor(
    Scale Alpha=&FloatFromCurve(FloatCurve=DataInterface() {propertyValues {Curve {keys [
         {tangentMode=RCTM_User, time=0, value=1}
         {arriveTangent=-0.0405, interpMode=RCIM_Cubic, leaveTangent=-0.0405, tangentMode=RCTM_User, time=0.7009, value=1}
         {arriveTangent=-8.0089, interpMode=RCIM_Cubic, leaveTangent=-8.0089, tangentMode=RCTM_User, time=1, value=0}
        ]}}})
    Scale RGB=&Lerp_Vector(
     A=&VectorFromFloat(Value=&FloatFromCurve(
       FloatCurve=DataInterface() {propertyValues {Curve {keys [<at 0 1>,<at 0.1 1>,<at 0.2 2>,<at 0.3 1>,<at 0.6 1>,<at 0.7 2>,<at 0.8 1>,<at 1 1>]}}}
       Scale Curve=&MaskFloatBySpawnGroup(Masked Float=10, Passthrough Float=5)
      ))
     B=&VectorFromFloat(Value=&RandomRangeFloat(Minimum=0.3, Maximum=5, Evaluation Type=::NewEnumerator1))
     Alpha=&FloatFromCurve(FloatCurve=DataInterface() {propertyValues {Curve {keys [
          {interpMode=RCIM_Cubic, leaveTangent=2.5294, tangentMode=RCTM_Break, time=0, value=0}
          {arriveTangent=-0.2166, interpMode=RCIM_Cubic, tangentMode=RCTM_Break, time=0.4, value=1}
         ]}}})
    )
    ScaleA=true
    ScaleRGB=true
   )
  ]
  rendererClasses [&NiagaraMeshRendererProperties]
  renderers [
   {rendererClass=&NiagaraMeshRendererProperties, rendererIndex=0}
   {rendererClass=&NiagaraMeshRendererProperties, rendererIndex=1}
   {rendererClass=&NiagaraMeshRendererProperties, rendererIndex=2}
   {rendererClass=&NiagaraMeshRendererProperties, rendererIndex=3}
  ]
 }
}
summary {
 systemName=NS_Emote_RainbowRaid_LightningSpawn
 emitters [
  {emitterName=Lightning, rendererClasses [&NiagaraMeshRendererProperties]}
  {emitterName=Burst, rendererClasses [&NiagaraLightRendererProperties, &NiagaraSpriteRendererProperties]}
  {emitterName=Corona, rendererClasses [&NiagaraSpriteRendererProperties]}
 ]
 userVariables []
}
