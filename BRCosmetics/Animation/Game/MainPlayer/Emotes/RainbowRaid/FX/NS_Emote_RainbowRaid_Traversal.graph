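# Niagara system for the RainbowRaid traversal emote, trailing glowing auroras, energy ribbons and splashes from character sockets read through a shared Socket Reader.
# Auroras use meshes, Core uses ribbons and Splashes uses sprites, with spawn groups and partitions varying placement and motion.
# Owner speed drives spawn rate, drag, velocity and world/local position blending, while front-end and time-of-day modifiers tune the look.
# Auroras001 and Auroras002 are disabled.
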NiagaraSystem /BRCosmetics/Animation/Game/MainPlayer/Emotes/RainbowRaid/FX/NS_Emote_RainbowRaid_Traversal
dependencies {
 usedDataInterfaces [
  &NiagaraDataInterfaceColorCurve
  &NiagaraDataInterfaceCurve
  &NiagaraDataInterfaceVectorCurve
  &NiagaraDataInterfaceArrayColor
  &NiagaraDataInterfaceVector2DCurve
  &NiagaraDataInterfaceSocketReader
 ]
 usedDynamicInputs []
 usedModules [
  &self:SystemSpawnScript.SystemScriptSource.SystemScriptGraph.NiagaraNodeAssignment_0.SetVariables_#1
  &self:SystemSpawnScript.SystemScriptSource.SystemScriptGraph.NiagaraNodeAssignment_1.SetVariables_#2
  &SystemState
  &Timeline
  &BlendSpaceHelper
 ]
}
emitters {
 Auroras {
  emitterSpawnScript [/SetVariables(Emitter.FE Particle Spawn Rate=60, Emitter.Idle Spawn Rate=15, Emitter.Max Traversal Speed=100, Emitter.Particle Spawn Rate=30) {isSetParameters=true, module=SetVariables_#3}]
  emitterUpdateScript [
   &EmitterState()
   &SpawnRate(SpawnGroup=1, SpawnRate=50, Use Spawn Probability=false) {module=SpawnRate001}
   &SpawnRate(
    Spawn Probability=0.8
    SpawnRate=&FN_FrontEndFloatModifier(
     Gameplay=&MakeCustomFloatFromBool(
      Bool=&SetBoolByFloatComparison(A=&VectorLength(Vector=$User.OwnerLocalVelocity), B=100)
      True Float=$Emitter.Particle Spawn Rate
      False Float=$Emitter.Idle Spawn Rate
     )
     Front End=$Emitter.FE Particle Spawn Rate
    )
    Use Spawn Probability=true
   )
  ]
  particleSpawnScript [
   &PartitionParticles()
   &InitializeParticle(
    Color=&MultiplyLinearColor(
     Linear Color=&RandomRangeLinearColor(Minimum=<rgba 5 5 5 8>, Maximum=<rgba 1 1 1 1>)
     Scale Factor=&MakeVector4_FromLinearColor(LinearColor=&FNTimeOfDayColor(Night Color=<rgba 1 1 1 1>, Morning Color=<rgba 1 1 1 1>, Day Color=<rgba 1 1 1 1.2>, Evening Color=<rgba 1 1 1 1>))
    )
    Lifetime=&MaskFloat(
     Mask Source=0
     Mask Parameter=?
     Mask=0
     Masked Float=&FN_FrontEndFloatModifier(
      Gameplay=&MakeCustomFloatFromBool(
       Bool=&SetBoolByFloatComparison(A=&VectorLength(Vector=$Engine.Owner.Velocity), B=50, Comparison Type=::NewEnumerator6)
       True Float=&RandomRangeFloat(Minimum=0.6, Maximum=0.4)
       False Float=&RandomRangeFloat(Minimum=1.2, Maximum=1)
      )
      Front End=&RandomRangeFloat(Minimum=1, Maximum=1.2)
     )
     Passthrough Float=&RandomRangeFloat(Minimum=1.2, Maximum=1)
    )
    Mesh Scale=&MaskVector(
     Mask Source=0
     Mask Parameter=?
     Mask=0
     Masked Vector=&RandomRangeVector(Minimum=<0.5 2 3>, Maximum=<1.5 2 7>)
     Passthrough Vector=&RandomRangeVector(Minimum=<0.5 1 3>, Maximum=<1.5 1 2>)
    )
    Mesh Scale Mode=::NewEnumerator4
    Sprite Rotation Mode=::NewEnumerator2
    Sprite Size Mode=::NewEnumerator1
    Uniform Sprite Size Max=50
    Uniform Sprite Size Min=30
    UsePositionOffset=false
    Write Material Random=true
   ) {module=InitializeParticle001}
   &SocketLocation(
    SamplePosOffset=true
    Sampled Position Offset=&MaskVector(
     Mask Source=0
     Mask Parameter=?
     Mask=0
     Masked Vector=&MaskVector(
      Mask Source=1
      Mask Parameter=$Particles.PartitionParticles.Partition
      Mask=0
      Masked Vector=&RandomRangeVector(Minimum=<5 -20 -30>, Maximum=<5 -20 -20>)
      Passthrough Vector=&RandomRangeVector(Minimum=<-5 -20 -30>, Maximum=<-5 -20 -20>)
     )
     Passthrough Vector=&MaskVector(Mask Source=1, Mask Parameter=$Particles.PartitionParticles.Partition, Mask=0, Masked Vector=<10 10 -15>, Passthrough Vector=<-10 10 -15>)
    )
    Sampled Position Offset Coordinate Space:ENiagaraSkelMeshTransforms=::NewEnumerator2
    Socket Index=$Particles.PartitionParticles.Partition
    Socket Reader=$System.Socket Reader
    Socket Sampling Mode=::NewEnumerator3
    VelocitySpeedLimit=false
   )
   &ShapeLocation(
    Box Size=<5 30 0>
    MaskSpawnGroup=true
    Offset=/Niagara/DynamicInputs/Helpers/MaskPosition(
     Mask Source:NiagaraInt32=1
     Mask Parameter:NiagaraInt32=$Particles.PartitionParticles.Partition
     Mask:NiagaraInt32=0
     Masked Position:NiagaraPosition=<-7 -5 -5>
     Passthrough Position:NiagaraPosition=<7 -5 -5>
    )
    Offset Coordinate Space=::Local
    Offset Mode=::NewEnumerator0
    Override Seed=false
    Shape Origin=$Output.SocketLocation.SampledPosition
    Shape Primitive=::NewEnumerator2
    Spawn Group:NiagaraInt32=&SpawnGroup()
    Spawn Group Mask=1
    Surface Only=false
    UseOwnerScale=true
   )
   &AddVelocity(
    Velocity=&RandomRangeVector(Minimum=<0 -2 0.8>, Maximum=<0 -2 0.2>)
    Velocity Speed Scale=&MaskFloatBySpawnGroup(
     Masked Float=&FN_FrontEndFloatModifier(
      Gameplay=&MakeCustomFloatFromBool(
       Bool=&SetBoolByFloatComparison(A=&VectorLength(Vector=$Engine.Owner.Velocity), B=50)
       True Float=50
       False Float=200
      )
      Front End=&RandomRangeFloat(Minimum=100, Maximum=400)
     )
     Passthrough Float=&FN_FrontEndFloatModifier(
      Gameplay=&MakeCustomFloatFromBool(
       Bool=&SetBoolByFloatComparison(A=&VectorLength(Vector=$Engine.Owner.Velocity), B=50)
       True Float=30
       False Float=20
      )
      Front End=&RandomRangeFloat(Minimum=1, Maximum=1)
     )
    )
   ) {module=AddVelocity001}
   &CameraOffset(Camera Offset Amount=&MaskFloatBySpawnGroup(Masked Float=15, Passthrough Float=10))
   &InitialMeshOrientation(
    Mesh Orientation Mode=::NewEnumerator2
    Rotation=&RandomRangeVector(Minimum=<-0.05 -0.05 0>, Maximum=<0.05 0.05 0>)
    Use Rotation Vector=true
   )
  ]
  particleUpdateScript [
   &ParticleState()
   &ScaleMeshSize(Scale Factor=&MaskVector(
     Mask Source=0
     Mask Parameter=?
     Mask=0
     Masked Vector=&VectorFromCurve(VectorCurve=DataInterface() {propertyValues {
        XCurve {keys [<at 0 1>,<at 1 1>]}
        YCurve {
         defaultValue=1
         keys [{interpMode=RCIM_Cubic, tangentMode=RCTM_User, time=0, value=1}, {interpMode=RCIM_Cubic, tangentMode=RCTM_User, time=1, value=1}]
        }
        ZCurve {
         defaultValue=1
         keys [
          {arriveTangent=4.7634, interpMode=RCIM_Cubic, leaveTangent=4.7634, tangentMode=RCTM_User, time=0.0083, value=0.634}
          {arriveTangent=0.0496, interpMode=RCIM_Cubic, leaveTangent=0.0496, tangentMode=RCTM_User, time=0.2114, value=1.1246}
          {interpMode=RCIM_Cubic, tangentMode=RCTM_User, time=1, value=0.5504}
         ]
        }
       }})
     Passthrough Vector=&VectorFromCurve(VectorCurve=DataInterface() {propertyValues {
        XCurve {keys [<at 0 1>,<at 1 1>]}
        YCurve {
         defaultValue=1
         keys [{interpMode=RCIM_Cubic, tangentMode=RCTM_User, time=0, value=1}, {interpMode=RCIM_Cubic, tangentMode=RCTM_User, time=1, value=1}]
        }
        ZCurve {
         defaultValue=1
         keys [
          {arriveTangent=0.0681, interpMode=RCIM_Cubic, leaveTangent=0.0681, tangentMode=RCTM_User, time=0.0083, value=0.634}
          {arriveTangent=0.0496, interpMode=RCIM_Cubic, leaveTangent=0.0496, tangentMode=RCTM_User, time=0.6193, value=1.1246}
          {interpMode=RCIM_Cubic, tangentMode=RCTM_User, time=1, value=0.5504}
         ]
        }
       }})
    ))
   &Drag(
    Drag=&FloatFromCurve(
     FloatCurve=DataInterface() {propertyValues {Curve {
        defaultValue=1
        keys [{interpMode=RCIM_Cubic, tangentMode=RCTM_User, time=0, value=0.75}, {interpMode=RCIM_Cubic, tangentMode=RCTM_User, time=0.65, value=1}]
       }}}
     Scale Curve=&Lerp_Float(
      A=20
      B=8
      Alpha=&RemapRange(
       Input Value To Remap=&VectorLength(Vector=$User.OwnerLocalVelocity)
       Input Min=0
       Input Max=$Emitter.Max Traversal Speed
       Output Min=0
       Output Max=1
      )
     )
    )
    Use Linear Drag=true
    Use Rotational Drag=true
   )
   &DynamicMaterialParameters(
    Index 0 Param 1=&Add_Float(
     A=&RandomRangeFloat(Minimum=0, Maximum=1)
     B=&FloatFromCurve(
      FloatCurve=DataInterface() {propertyValues {Curve {keys [
          <at 0 0 RCIM_Cubic>
          {arriveTangent=1.5947, interpMode=RCIM_Cubic, leaveTangent=1.5947, tangentMode=RCTM_User, time=0.2557, value=0.5257}
          <at 1 1 RCIM_Cubic>
         ]}}}
      CurveIndex=$Particles.Age
      Scale Curve=0.25
     )
    )
    Module.Index0Float0Write=true
    Module.Index0Float1Write=true
    Module.Index0Float2Write=true
    Module.Index0Float3Write=true
   ) {module=DynamicMaterialParameters001}
   /SetVariables(Particles.BrightColor:LinearColor=&MultiplyLinearColor(
     Linear Color=&MultiplyLinearColor(
      Linear Color=&ColorFromCurve(ColorCurve=DataInterface() {propertyValues {
         AlphaCurve {keys [<at 0 0 RCIM_Cubic>,<at 0.0289 1>,<at 1 0 RCIM_Cubic>]}
         BlueCurve {keys [<at 0 14.0903 RCIM_Cubic>,<at 1 1 RCIM_Cubic>]}
         GreenCurve {keys [<at 0 15 RCIM_Cubic>,<at 1 0.1803 RCIM_Cubic>]}
         RedCurve {keys [<at 0 0.4863 RCIM_Cubic>,<at 1 0.0307 RCIM_Cubic>]}
        }})
      Scale Factor=&MakeVector4_FromLinearColor(LinearColor=&FNTimeOfDayColor(Night Color=<rgba 1 1 1 1>, Morning Color=<rgba 1 1 1 1>, Day Color=<rgba 2 2 2 1>, Evening Color=<rgba 1 1 1 1>))
     )
     Scale Factor=&MakeVector4_FromLinearColor(LinearColor=<rgba 3 3 3 1>)
    )) {isSetParameters=true, module=SetVariables_#4}
   &ScaleColor(
    Scale Alpha=&MaskFloatBySpawnGroup(
     Masked Float=&FloatFromCurve(
      FloatCurve=DataInterface() {propertyValues {Curve {keys [
          <at 0 0 RCIM_Cubic>
          {arriveTangent=0.037, interpMode=RCIM_Cubic, leaveTangent=0.037, tangentMode=RCTM_User, time=0.3, value=1}
          {arriveTangent=-1.4864, interpMode=RCIM_Cubic, leaveTangent=-1.4864, tangentMode=RCTM_User, time=0.5126, value=0.1699}
          <at 1 0 RCIM_Cubic>
         ]}}}
      Scale Curve=&Multiply_Float(
       A=&Multiply_Float(
        A=&RandomRangeFloat(Minimum=10, Maximum=5)
        B=&FNTimeOfDayScalar(Night Scale=1, Morning Scale=1, Evening Scale=1)
       )
       B=&Multiply_Float(A=&FN_FrontEndFloatModifier(Gameplay=0.3, Front End=0.1), B=1)
      )
     )
     Passthrough Float=&FloatFromCurve(
      FloatCurve=DataInterface() {propertyValues {Curve {keys [
          {arriveTangent=7.7883, interpMode=RCIM_Cubic, leaveTangent=7.7883, tangentMode=RCTM_User, time=0, value=0}
          {arriveTangent=0.037, interpMode=RCIM_Cubic, leaveTangent=0.037, tangentMode=RCTM_User, time=0.3, value=1}
          {arriveTangent=-1.4864, interpMode=RCIM_Cubic, leaveTangent=-1.4864, tangentMode=RCTM_User, time=0.5371, value=0.3812}
          <at 1 0 RCIM_Cubic>
         ]}}}
      Scale Curve=&Multiply_Float(
       A=&Multiply_Float(
        A=&RandomRangeFloat(Minimum=10, Maximum=20)
        B=&FNTimeOfDayScalar(Night Scale=1, Morning Scale=1, Evening Scale=1)
       )
       B=&Multiply_Float(A=&FN_FrontEndFloatModifier(Gameplay=1, Front End=0.5), B=1)
      )
     )
    )
    Scale RGB=&MakeVectorFromLinearColorRGB(Color=&ColorFromCurve(ColorCurve=DataInterface() {propertyValues {
        AlphaCurve {keys [<at 0 0 RCIM_Cubic>,<at 0.0218 3>,<at 0.135 1>,<at 1 0>,<at 1 0>]}
        BlueCurve {keys [<at 0 2.9148 RCIM_Cubic>,<at 0.1983 0.933>,<at 0.9955 1>]}
        GreenCurve {keys [<at 0 3 RCIM_Cubic>,<at 0.1983 1>,<at 0.9955 0.0022>]}
        RedCurve {keys [<at 0 0.2521 RCIM_Cubic>,<at 0.1983 0.2916>,<at 0.9955 0.8069>]}
       }}))
    ScaleA=true
    ScaleRGB=true
   ) {module=ScaleColor001}
   &BlendSpace("Position Blend (World <-> Local)":NiagaraFloat=&FN_FrontEndFloatModifier(
     Gameplay=&MaskFloat(
      Mask Source=0
      Mask Parameter=?
      Mask=0
      Masked Float=&MakeCustomFloatFromBool(
       Bool=&SetBoolByFloatComparison(A=&VectorLength(Vector=$Engine.Owner.Velocity), B=400)
       True Float=1
       False Float=0
      )
      Passthrough Float=1
     )
     Front End=1
     Translation Contribution:NiagaraFloat=1
     Rotation Contribution:NiagaraFloat=1
    ))
   &SolveForcesAndVelocity(Clamp Velocity=false, Limit Acceleration=false)
  ]
  rendererClasses [&NiagaraMeshRendererProperties]
  renderers [
   {rendererClass=&NiagaraMeshRendererProperties, rendererIndex=0}
   {rendererClass=&NiagaraMeshRendererProperties, rendererIndex=1}
  ]
 }
 Auroras001 {
  bEnabled=false
  emitterSpawnScript [/SetVariables(Emitter.FE Particle Spawn Rate=60, Emitter.Idle Spawn Rate=50, Emitter.Max Traversal Speed=100, Emitter.Particle Spawn Rate=50) {isSetParameters=true, module=SetVariables_#3}]
  emitterUpdateScript [
   &EmitterState()
   &SpawnRate(
    Spawn Probability=0.8
    SpawnRate=&FN_FrontEndFloatModifier(
     Gameplay=&MakeCustomFloatFromBool(
      Bool=&SetBoolByFloatComparison(A=&VectorLength(Vector=$User.OwnerLocalVelocity), B=100)
      True Float=$Emitter.Particle Spawn Rate
      False Float=$Emitter.Idle Spawn Rate
     )
     Front End=$Emitter.FE Particle Spawn Rate
    )
    Use Spawn Probability=true
   )
  ]
  particleSpawnScript [
   &PartitionParticles()
   &InitializeParticle(
    Color=&MultiplyLinearColor(
     Linear Color=&RandomRangeLinearColor(Minimum=<rgba 5 5 5 8>, Maximum=<rgba 1 1 1 1>)
     Scale Factor=&MakeVector4_FromLinearColor(LinearColor=&FNTimeOfDayColor(Night Color=<rgba 1 1 1 1>, Morning Color=<rgba 1 1 1 1>, Day Color=<rgba 1 1 1 1.2>, Evening Color=<rgba 1 1 1 1>))
    )
    Lifetime=&RandomRangeFloat(Minimum=1.2, Maximum=1)
    Mesh Scale Max=<1.5 1 2>
    Mesh Scale Min=<0.5 1 3>
    Mesh Scale Mode=::NewEnumerator2
    Sprite Rotation Mode=::NewEnumerator2
    Sprite Size Mode=::NewEnumerator1
    Uniform Sprite Size Max=50
    Uniform Sprite Size Min=30
    UsePositionOffset=false
    Write Material Random=true
   ) {module=InitializeParticle001}
   &SocketLocation(
    SamplePosOffset=true
    Sampled Position Offset=&MaskVector(Mask Source=1, Mask Parameter=$Particles.PartitionParticles.Partition, Mask=0, Masked Vector=<10 10 -15>, Passthrough Vector=<-10 10 -15>)
    Sampled Position Offset Coordinate Space:ENiagaraSkelMeshTransforms=::NewEnumerator2
    Socket Index=$Particles.PartitionParticles.Partition
    Socket Reader=$System.Socket Reader
    Socket Sampling Mode=::NewEnumerator3
    VelocitySpeedLimit=false
   )
   &ShapeLocation(
    Box Size=<5 30 0>
    MaskSpawnGroup=false
    Offset=&MaskPosition(
     Mask Source:NiagaraInt32=1
     Mask Parameter:NiagaraInt32=$Particles.PartitionParticles.Partition
     Mask:NiagaraInt32=0
     Masked Position:NiagaraPosition=<-7 -5 -5>
     Passthrough Position:NiagaraPosition=<7 -5 -5>
    )
    Offset Coordinate Space=::Local
    Offset Mode=::NewEnumerator0
    Override Seed=false
    Shape Origin=$Output.SocketLocation.SampledPosition
    Shape Primitive=::NewEnumerator2
    Surface Only=false
    UseOwnerScale=true
   )
   &AddVelocity(
    Velocity=&RandomRangeVector(Minimum=<0 -2 0.8>, Maximum=<0 -2 0.2>)
    Velocity Speed Scale=&FN_FrontEndFloatModifier(
     Gameplay=&MakeCustomFloatFromBool(
      Bool=&SetBoolByFloatComparison(A=&VectorLength(Vector=$Engine.Owner.Velocity), B=50)
      True Float=30
      False Float=20
     )
     Front End=&RandomRangeFloat(Minimum=1, Maximum=1)
    )
   ) {module=AddVelocity001}
   &CameraOffset(Camera Offset Amount=10)
   &InitialMeshOrientation(
    Mesh Orientation Mode=::NewEnumerator2
    Rotation=&RandomRangeVector(Minimum=<-0.05 -0.05 0>, Maximum=<0.05 0.05 0>)
    Use Rotation Vector=true
   )
  ]
  particleUpdateScript [
   &ParticleState()
   &ScaleMeshSize(Scale Factor=&VectorFromCurve(VectorCurve=DataInterface() {propertyValues {
       XCurve {keys [<at 0 1>,<at 1 1>]}
       YCurve {
        defaultValue=1
        keys [{interpMode=RCIM_Cubic, tangentMode=RCTM_User, time=0, value=1}, {interpMode=RCIM_Cubic, tangentMode=RCTM_User, time=1, value=1}]
       }
       ZCurve {
        defaultValue=1
        keys [
         {arriveTangent=0.0681, interpMode=RCIM_Cubic, leaveTangent=0.0681, tangentMode=RCTM_User, time=0.0083, value=0.634}
         {arriveTangent=0.0496, interpMode=RCIM_Cubic, leaveTangent=0.0496, tangentMode=RCTM_User, time=0.6193, value=1.1246}
         {interpMode=RCIM_Cubic, tangentMode=RCTM_User, time=1, value=0.5504}
        ]
       }
      }}))
   &Drag(
    Drag=&FloatFromCurve(
     FloatCurve=DataInterface() {propertyValues {Curve {
        defaultValue=1
        keys [{interpMode=RCIM_Cubic, tangentMode=RCTM_User, time=0, value=0.75}, {interpMode=RCIM_Cubic, tangentMode=RCTM_User, time=0.65, value=1}]
       }}}
     Scale Curve=&Lerp_Float(
      A=20
      B=8
      Alpha=&RemapRange(
       Input Value To Remap=&VectorLength(Vector=$User.OwnerLocalVelocity)
       Input Min=0
       Input Max=$Emitter.Max Traversal Speed
       Output Min=0
       Output Max=1
      )
     )
    )
    Use Linear Drag=true
    Use Rotational Drag=true
   )
   &DynamicMaterialParameters(
    Index 0 Param 1=&Add_Float(
     A=&RandomRangeFloat(Minimum=0, Maximum=1)
     B=&FloatFromCurve(
      FloatCurve=DataInterface() {propertyValues {Curve {keys [
          <at 0 0 RCIM_Cubic>
          {arriveTangent=1.5947, interpMode=RCIM_Cubic, leaveTangent=1.5947, tangentMode=RCTM_User, time=0.2557, value=0.5257}
          <at 1 1 RCIM_Cubic>
         ]}}}
      CurveIndex=$Particles.Age
      Scale Curve=0.25
     )
    )
    Module.Index0Float0Write=true
    Module.Index0Float1Write=true
    Module.Index0Float2Write=true
    Module.Index0Float3Write=true
   ) {module=DynamicMaterialParameters001}
   /SetVariables(Particles.BrightColor:LinearColor=&MultiplyLinearColor(
     Linear Color=&MultiplyLinearColor(
      Linear Color=&ColorFromCurve(ColorCurve=DataInterface() {propertyValues {
         AlphaCurve {keys [<at 0 0 RCIM_Cubic>,<at 0.0289 1>,<at 1 0 RCIM_Cubic>]}
         BlueCurve {keys [<at 0 14.0903 RCIM_Cubic>,<at 1 1 RCIM_Cubic>]}
         GreenCurve {keys [<at 0 15 RCIM_Cubic>,<at 1 0.1803 RCIM_Cubic>]}
         RedCurve {keys [<at 0 0.4863 RCIM_Cubic>,<at 1 0.0307 RCIM_Cubic>]}
        }})
      Scale Factor=&MakeVector4_FromLinearColor(LinearColor=&FNTimeOfDayColor(Night Color=<rgba 1 1 1 1>, Morning Color=<rgba 1 1 1 1>, Day Color=<rgba 2 2 2 1>, Evening Color=<rgba 1 1 1 1>))
     )
     Scale Factor=&MakeVector4_FromLinearColor(LinearColor=<rgba 5 5 5 1>)
    )) {isSetParameters=true, module=SetVariables_#4}
   &ScaleColor(
    Scale Alpha=&FloatFromCurve(
     FloatCurve=DataInterface() {propertyValues {Curve {keys [
         {arriveTangent=7.7883, interpMode=RCIM_Cubic, leaveTangent=7.7883, tangentMode=RCTM_User, time=0, value=0}
         {arriveTangent=0.037, interpMode=RCIM_Cubic, leaveTangent=0.037, tangentMode=RCTM_User, time=0.3, value=1}
         {arriveTangent=-1.4864, interpMode=RCIM_Cubic, leaveTangent=-1.4864, tangentMode=RCTM_User, time=0.5371, value=0.3812}
         <at 1 0 RCIM_Cubic>
        ]}}}
     Scale Curve=&Multiply_Float(
      A=&Multiply_Float(
       A=&RandomRangeFloat(Minimum=10, Maximum=20)
       B=&FNTimeOfDayScalar(Night Scale=1, Morning Scale=1, Evening Scale=1)
      )
      B=&Multiply_Float(A=&FN_FrontEndFloatModifier(Gameplay=1, Front End=0.5), B=1)
     )
    )
    Scale RGB=&MakeVectorFromLinearColorRGB(Color=&ColorFromCurve(ColorCurve=DataInterface() {propertyValues {
        AlphaCurve {keys [<at 0 0 RCIM_Cubic>,<at 0.0218 3>,<at 0.135 1>,<at 1 0>,<at 1 0>]}
        BlueCurve {keys [<at 0 4.8581 RCIM_Cubic>,<at 0.1983 1.2129>,<at 0.9955 1>]}
        GreenCurve {keys [<at 0 5 RCIM_Cubic>,<at 0.1983 1.3>,<at 0.9955 0.0022>]}
        RedCurve {keys [<at 0 0.4201 RCIM_Cubic>,<at 0.1983 0.3791>,<at 0.9955 0.8069>]}
       }}))
    ScaleA=true
    ScaleRGB=true
   ) {module=ScaleColor001}
   &SolveForcesAndVelocity(Clamp Velocity=false, Limit Acceleration=false)
  ]
  rendererClasses [&NiagaraMeshRendererProperties]
  renderers [
   {rendererClass=&NiagaraMeshRendererProperties, rendererIndex=0}
   {rendererClass=&NiagaraMeshRendererProperties, rendererIndex=1}
  ]
 }
 Auroras002 {
  bEnabled=false
  emitterSpawnScript [/SetVariables(Emitter.FE Particle Spawn Rate=60, Emitter.Idle Spawn Rate=15, Emitter.Max Traversal Speed=100, Emitter.Particle Spawn Rate=30) {isSetParameters=true, module=SetVariables_#3}]
  emitterUpdateScript [
   &EmitterState()
   &SpawnRate(
    Spawn Probability=0.8
    SpawnRate=&FN_FrontEndFloatModifier(
     Gameplay=&MakeCustomFloatFromBool(
      Bool=&SetBoolByFloatComparison(A=&VectorLength(Vector=$User.OwnerLocalVelocity), B=100)
      True Float=$Emitter.Particle Spawn Rate
      False Float=$Emitter.Idle Spawn Rate
     )
     Front End=$Emitter.FE Particle Spawn Rate
    )
    Use Spawn Probability=true
   )
  ]
  particleSpawnScript [
   &PartitionParticles()
   &InitializeParticle(
    Color=&MultiplyLinearColor(
     Linear Color=&RandomRangeLinearColor(Minimum=<rgba 5 5 5 8>, Maximum=<rgba 1 1 1 1>)
     Scale Factor=&MakeVector4_FromLinearColor(LinearColor=&FNTimeOfDayColor(Night Color=<rgba 1 1 1 1>, Morning Color=<rgba 1 1 1 1>, Day Color=<rgba 1 1 1 1.2>, Evening Color=<rgba 1 1 1 1>))
    )
    Lifetime=&MakeCustomFloatFromBool(
     Bool=&SetBoolByFloatComparison(A=&VectorLength(Vector=$Engine.Owner.Velocity), B=50, Comparison Type=::NewEnumerator6)
     True Float=&RandomRangeFloat(Minimum=0.6, Maximum=0.4)
     False Float=&RandomRangeFloat(Minimum=1.2, Maximum=1)
    )
    Mesh Scale Max=<1.5 2 7>
    Mesh Scale Min=<0.5 2 3>
    Mesh Scale Mode=::NewEnumerator2
    Sprite Rotation Mode=::NewEnumerator2
    Sprite Size Mode=::NewEnumerator1
    Uniform Sprite Size Max=50
    Uniform Sprite Size Min=30
    UsePositionOffset=false
    Write Material Random=true
   ) {module=InitializeParticle001}
   &SocketLocation(
    SamplePosOffset=true
    Sampled Position Offset=&MaskVector(
     Mask Source=1
     Mask Parameter=$Particles.PartitionParticles.Partition
     Mask=0
     Masked Vector=&RandomRangeVector(Minimum=<5 -20 -30>, Maximum=<5 -20 -20>)
     Passthrough Vector=&RandomRangeVector(Minimum=<-5 -20 -30>, Maximum=<-5 -20 -20>)
    )
    Sampled Position Offset Coordinate Space:ENiagaraSkelMeshTransforms=::NewEnumerator2
    Socket Index=$Particles.PartitionParticles.Partition
    Socket Reader=$System.Socket Reader
    Socket Sampling Mode=::NewEnumerator3
    VelocitySpeedLimit=false
   )
   &AddVelocity(
    Velocity=&RandomRangeVector(Minimum=<0 -2 0.8>, Maximum=<0 -2 0.2>)
    Velocity Speed Scale=&FN_FrontEndFloatModifier(
     Gameplay=&MakeCustomFloatFromBool(
      Bool=&SetBoolByFloatComparison(A=&VectorLength(Vector=$Engine.Owner.Velocity), B=50)
      True Float=50
      False Float=200
     )
     Front End=&RandomRangeFloat(Minimum=100, Maximum=200)
    )
   ) {module=AddVelocity001}
   &CameraOffset(Camera Offset Amount=15)
   &InitialMeshOrientation(
    Mesh Orientation Mode=::NewEnumerator2
    Rotation=&RandomRangeVector(Minimum=<-0.05 -0.05 0>, Maximum=<0.05 0.05 0>)
    Use Rotation Vector=true
   )
  ]
  particleUpdateScript [
   &ParticleState()
   &ScaleMeshSize(Scale Factor=&VectorFromCurve(VectorCurve=DataInterface() {propertyValues {
       XCurve {keys [<at 0 1>,<at 1 1>]}
       YCurve {
        defaultValue=1
        keys [{interpMode=RCIM_Cubic, tangentMode=RCTM_User, time=0, value=1}, {interpMode=RCIM_Cubic, tangentMode=RCTM_User, time=1, value=1}]
       }
       ZCurve {
        defaultValue=1
        keys [
         {arriveTangent=4.7634, interpMode=RCIM_Cubic, leaveTangent=4.7634, tangentMode=RCTM_User, time=0.0083, value=0.634}
         {arriveTangent=0.0496, interpMode=RCIM_Cubic, leaveTangent=0.0496, tangentMode=RCTM_User, time=0.2114, value=1.1246}
         {interpMode=RCIM_Cubic, tangentMode=RCTM_User, time=1, value=0.5504}
        ]
       }
      }}))
   &Drag(
    Drag=&FloatFromCurve(
     FloatCurve=DataInterface() {propertyValues {Curve {
        defaultValue=1
        keys [{interpMode=RCIM_Cubic, tangentMode=RCTM_User, time=0, value=0.75}, {interpMode=RCIM_Cubic, tangentMode=RCTM_User, time=0.65, value=1}]
       }}}
     Scale Curve=&Lerp_Float(
      A=20
      B=8
      Alpha=&RemapRange(
       Input Value To Remap=&VectorLength(Vector=$User.OwnerLocalVelocity)
       Input Min=0
       Input Max=$Emitter.Max Traversal Speed
       Output Min=0
       Output Max=1
      )
     )
    )
    Use Linear Drag=true
    Use Rotational Drag=true
   )
   &DynamicMaterialParameters(
    Index 0 Param 1=&Add_Float(
     A=&RandomRangeFloat(Minimum=0, Maximum=1)
     B=&FloatFromCurve(
      FloatCurve=DataInterface() {propertyValues {Curve {keys [
          <at 0 0 RCIM_Cubic>
          {arriveTangent=1.5947, interpMode=RCIM_Cubic, leaveTangent=1.5947, tangentMode=RCTM_User, time=0.2557, value=0.5257}
          <at 1 1 RCIM_Cubic>
         ]}}}
      CurveIndex=$Particles.Age
      Scale Curve=0.25
     )
    )
    Module.Index0Float0Write=true
    Module.Index0Float1Write=true
    Module.Index0Float2Write=true
    Module.Index0Float3Write=true
   ) {module=DynamicMaterialParameters001}
   /SetVariables(Particles.BrightColor:LinearColor=&MultiplyLinearColor(
     Linear Color=&MultiplyLinearColor(
      Linear Color=&ColorFromCurve(ColorCurve=DataInterface() {propertyValues {
         AlphaCurve {keys [<at 0 0 RCIM_Cubic>,<at 0.0289 1>,<at 1 0 RCIM_Cubic>]}
         BlueCurve {keys [<at 0 14.0903 RCIM_Cubic>,<at 1 1 RCIM_Cubic>]}
         GreenCurve {keys [<at 0 15 RCIM_Cubic>,<at 1 0.1803 RCIM_Cubic>]}
         RedCurve {keys [<at 0 0.4863 RCIM_Cubic>,<at 1 0.0307 RCIM_Cubic>]}
        }})
      Scale Factor=&MakeVector4_FromLinearColor(LinearColor=&FNTimeOfDayColor(Night Color=<rgba 1 1 1 1>, Morning Color=<rgba 1 1 1 1>, Day Color=<rgba 2 2 2 1>, Evening Color=<rgba 1 1 1 1>))
     )
     Scale Factor=&MakeVector4_FromLinearColor(LinearColor=<rgba 3 3 3 1>)
    )) {isSetParameters=true, module=SetVariables_#4}
   &ScaleColor(
    Scale Alpha=&FloatFromCurve(
     FloatCurve=DataInterface() {propertyValues {Curve {keys [
         <at 0 0 RCIM_Cubic>
         {arriveTangent=0.037, interpMode=RCIM_Cubic, leaveTangent=0.037, tangentMode=RCTM_User, time=0.3, value=1}
         {arriveTangent=-1.4864, interpMode=RCIM_Cubic, leaveTangent=-1.4864, tangentMode=RCTM_User, time=0.5126, value=0.1699}
         <at 1 0 RCIM_Cubic>
        ]}}}
     Scale Curve=&Multiply_Float(
      A=&Multiply_Float(
       A=&RandomRangeFloat(Minimum=10, Maximum=5)
       B=&FNTimeOfDayScalar(Night Scale=1, Morning Scale=1, Evening Scale=1)
      )
      B=&Multiply_Float(A=&FN_FrontEndFloatModifier(Gameplay=1, Front End=0.5), B=1)
     )
    )
    Scale RGB=&MakeVectorFromLinearColorRGB(Color=&ColorFromCurve(ColorCurve=DataInterface() {propertyValues {
        AlphaCurve {keys [<at 0 0 RCIM_Cubic>,<at 0.0218 3>,<at 0.135 1>,<at 1 0>,<at 1 0>]}
        BlueCurve {keys [<at 0 2.9148 RCIM_Cubic>,<at 0.1983 0.933>,<at 0.9955 1>]}
        GreenCurve {keys [<at 0 3 RCIM_Cubic>,<at 0.1983 1>,<at 0.9955 0.0022>]}
        RedCurve {keys [<at 0 0.2521 RCIM_Cubic>,<at 0.1983 0.2916>,<at 0.9955 0.8069>]}
       }}))
    ScaleA=true
    ScaleRGB=true
   ) {module=ScaleColor001}
   &BlendSpace("Position Blend (World <-> Local)":NiagaraFloat=&MakeCustomFloatFromBool(
     Bool=&SetBoolByFloatComparison(A=&VectorLength(Vector=$Engine.Owner.Velocity), B=400)
     True Float=1
     False Float=0
     Translation Contribution:NiagaraFloat=1
     Rotation Contribution:NiagaraFloat=1
    ))
   &SolveForcesAndVelocity(Clamp Velocity=false, Limit Acceleration=false)
  ]
  rendererClasses [&NiagaraMeshRendererProperties]
  renderers [
   {rendererClass=&NiagaraMeshRendererProperties, rendererIndex=0}
   {rendererClass=&NiagaraMeshRendererProperties, rendererIndex=1}
  ]
 }
 Core {
  emitterSpawnScript [/SetVariables(Emitter.NewSpawnRate=100) {isSetParameters=true}]
  emitterUpdateScript [
   &EmitterState(Life Cycle Mode=::NewEnumerator1, Loop Behavior=::NewEnumerator1, Loop Duration Mode=::NewEnumerator1, UseLoopDelay=false)
   &SpawnRate(SpawnGroup=1, SpawnRate=$Emitter.NewSpawnRate, Use Spawn Probability=false) {module=SpawnRate001}
   &SpawnRate(SpawnRate=$Emitter.NewSpawnRate, Use Spawn Probability=false)
  ]
  particleSpawnScript [
   /SetVariables(Particles.RibbonID=/Niagara/DynamicInputs/Meshes/MakeRibbonIDFromSpawnGroup()) {isSetParameters=true}
   &InitializeParticle(
    Color=<rgba 0.8748 1.1847 1.5 1.5>
    Lifetime=&FN_FrontEndFloatModifier(Gameplay=0.6, Front End=0.8)
    Ribbon Width=45
    Ribbon Width Mode=::NewEnumerator1
    UsePositionOffset=false
    Write Material Random=true
   )
   &SocketLocation(
    SamplePosOffset=true
    Sampled Position Offset=<0 10 -10>
    Socket Index=&SpawnGroup()
    Socket Reader=$System.Socket Reader
    Socket Sampling Mode=::NewEnumerator3
    VelocitySpeedLimit=false
   )
   &AddVelocity(Velocity=<0 -60 0>, Velocity Speed Scale=2.5) {module=AddVelocity001}
   /SetVariables(Particles.RibbonFacing=$Engine.Owner.SystemZAxis) {isSetParameters=true}
  ]
  particleUpdateScript [
   &ParticleState()
   &SolveForcesAndVelocity(Clamp Velocity=false, Limit Acceleration=false)
   &ScaleColor(
    Scale Alpha=&FloatFromCurve(
     FloatCurve=DataInterface() {propertyValues {Curve {keys [<at 0 0>,<at 0.0777 1>,<at 1 0>]}}}
     Scale Curve=&FN_FrontEndFloatModifier(Gameplay=0.5, Front End=0.3)
    )
    Scale RGB=&VectorFromFloat(Value=&FloatFromCurve(FloatCurve=DataInterface() {propertyValues {Curve {keys [<at 0 1>,<at 0.3 0.8>]}}}))
    ScaleA=true
    ScaleRGB=true
   )
   /SetVariables(Particles.EnergyColor:LinearColor=&MakeCustomLinearColorFromBool(
     Bool=&Boolean_OrOperation(Bool1=$User.bIsFrontEnd, Bool2=$User.bIsFrontEndPreview)
     True Color=&ColorFromCurve(ColorCurve=DataInterface() {propertyValues {
        AlphaCurve {keys [<at 0 0.8 RCIM_Cubic>,<at 0.3618 0.2905>,<at 0.5949 0.0549>,<at 0.8289 0 RCIM_Cubic>]}
        BlueCurve {keys [<at 0 2.2507 RCIM_Cubic>,<at 0.1874 0.5056 RCIM_Cubic>]}
        GreenCurve {keys [<at 0 3 RCIM_Cubic>,<at 0.1874 1 RCIM_Cubic>]}
        RedCurve {keys [<at 0 0.6406 RCIM_Cubic>,<at 0.1874 0.0061 RCIM_Cubic>]}
       }})
     False Color=&MultiplyLinearColor(
      Linear Color=&MultiplyLinearColor(
       Linear Color=&LerpLinearColors(
        StartColor=&ColorFromCurve(ColorCurve=DataInterface() {propertyValues {
           AlphaCurve {keys [<at 0 0.8 RCIM_Cubic>,<at 0.8289 0 RCIM_Cubic>]}
           BlueCurve {keys [<at 0 2.2507 RCIM_Cubic>,<at 0.4337 0.5056 RCIM_Cubic>]}
           GreenCurve {keys [<at 0 3 RCIM_Cubic>,<at 0.4337 1 RCIM_Cubic>]}
           RedCurve {keys [<at 0 0.6406 RCIM_Cubic>,<at 0.4337 0.0061 RCIM_Cubic>]}
          }})
        EndColor=&ColorFromCurve(ColorCurve=DataInterface() {propertyValues {
           AlphaCurve {keys [<at 0 0.8 RCIM_Cubic>,<at 0.3075 0.8266>,<at 0.7036 0.5665>,<at 1 0 RCIM_Cubic>]}
           BlueCurve {keys [<at 0 2.2507 RCIM_Cubic>,<at 0.4337 0.5056 RCIM_Cubic>]}
           GreenCurve {keys [<at 0 3 RCIM_Cubic>,<at 0.4337 1 RCIM_Cubic>]}
           RedCurve {keys [<at 0 0.6406 RCIM_Cubic>,<at 0.4337 0.0061 RCIM_Cubic>]}
          }})
        LerpFactor=$System.Timeline.NormalizedTime
       )
       Scale Factor=&MakeVector4(
        X=1
        Y=1
        Z=1
        W=&FNTimeOfDayScalar(Night Scale=0.8, Morning Scale=0.7, Day Scale=0.5, Evening Scale=0.7)
       )
      )
      Scale Factor=&MakeVector4(
       X=1
       Y=1
       Z=1
       W=&FloatFromCurve(
        FloatCurve=DataInterface() {propertyValues {Curve {keys [<at 0.0834 0>,<at 1 1>]}}}
        CurveIndex=&ReturnNormalizedExecIndex()
       )
      )
     )
    )) {isSetParameters=true}
   &DynamicMaterialParameters(
    Index 0 Param 1=$Engine.Time
    Module.Index0Float0Write=true
    Module.Index0Float1Write=false
    Module.Index0Float2Write=false
    Module.Index0Float3Write=false
   )
   &BlendSpace("Position Blend (World <-> Local)":NiagaraFloat=&FN_FrontEndFloatModifier(
     Gameplay=&MakeCustomFloatFromBool(
      Bool=&SetBoolByFloatComparison(A=&VectorLength(Vector=$Engine.Owner.Velocity), B=400)
      True Float=1
      False Float=0.2
     )
     Front End=1
     Translation Contribution:NiagaraFloat=1
     Rotation Contribution:NiagaraFloat=1
    ))
  ]
  rendererClasses [&NiagaraRibbonRendererProperties]
  renderers [{rendererClass=&NiagaraRibbonRendererProperties}]
 }
 Splashes {
  emitterSpawnScript [/SetVariables(Emitter.FE Particle Spawn Rate=60, Emitter.Idle Spawn Rate=30, Emitter.Max Traversal Speed=100, Emitter.Particle Spawn Rate=50) {isSetParameters=true}]
  emitterUpdateScript [
   &EmitterState(Inactive Response=::NewEnumerator1, Life Cycle Mode=::NewEnumerator1, Loop Behavior=::NewEnumerator1, Loop Duration=3600, UseLoopDelay=false)
   &SpawnRate(SpawnGroup=2, SpawnRate=100, Use Spawn Probability=false) {module=SpawnRate002}
   &SpawnRate(
    Spawn Probability=0.8
    SpawnGroup=1
    SpawnRate=&FN_FrontEndFloatModifier(
     Gameplay=&MakeCustomFloatFromBool(
      Bool=&SetBoolByFloatComparison(A=&VectorLength(Vector=$User.OwnerLocalVelocity), B=100)
      True Float=$Emitter.Particle Spawn Rate
      False Float=$Emitter.Idle Spawn Rate
     )
     Front End=$Emitter.FE Particle Spawn Rate
    )
    Use Spawn Probability=true
   ) {module=SpawnRate001}
   &SpawnRate(
    Spawn Probability=0.8
    SpawnRate=&FN_FrontEndFloatModifier(
     Gameplay=&MakeCustomFloatFromBool(
      Bool=&SetBoolByFloatComparison(A=&VectorLength(Vector=$User.OwnerLocalVelocity), B=100)
      True Float=$Emitter.Particle Spawn Rate
      False Float=$Emitter.Idle Spawn Rate
     )
     Front End=$Emitter.FE Particle Spawn Rate
    )
    Use Spawn Probability=true
   )
  ]
  particleSpawnScript [
   &PartitionParticles()
   &InitializeParticle(
    Color=&SelectLinearColorFromArray(
     Array Sampling Mode=::NewEnumerator0
     Direct Array Index=?
     Color Selection Array=DataInterface() {propertyValues {ColorData [<rgba 0.2737 1 0.6739 0.6889>, <rgba 10 0.9373 7.5246 0.6889>]}}
     Random Seed=0
    )
    Lifetime Max=0.4
    Lifetime Min=0.3
    Lifetime Mode=::NewEnumerator1
    Mesh Scale Mode=::NewEnumerator3
    Sprite Rotation Mode=::NewEnumerator2
    Sprite Size Mode=::NewEnumerator1
    Uniform Sprite Size Max=&MaskFloat(Mask Source=0, Mask Parameter=?, Mask=2, Masked Float=80, Passthrough Float=70)
    Uniform Sprite Size Min=30
    UsePositionOffset=false
    Write Material Random=true
   ) {module=InitializeParticle001}
   &SocketLocation(
    SamplePosOffset=true
    Sampled Position Offset=&MaskVector(
     Mask Source=1
     Mask Parameter=$Particles.PartitionParticles.Partition
     Mask=0
     Masked Vector=&RandomRangeVector(Minimum=<5 -15 -10>, Maximum=<5 -15 0>)
     Passthrough Vector=&RandomRangeVector(Minimum=<-5 -15 -10>, Maximum=<-5 -15 0>)
    )
    Socket Index=$Particles.PartitionParticles.Partition
    Socket Reader=$System.Socket Reader
    Socket Sampling Mode=::NewEnumerator3
    VelocitySpeedLimit=false
   )
   &ShapeLocation(
    Box Size=&MaskVectorBySpawnGroup(Masked Vector=<35 45 0>, Passthrough Vector=<40 1 1>, Spawn Group Mask=2)
    MaskSpawnGroup=true
    Offset=&MaskVector(Mask Source=0, Mask Parameter=?, Mask=2, Masked Vector=<0 10 -10>, Passthrough Vector=<0 0 0>)
    Offset Coordinate Space=::Local
    Offset Mode=::NewEnumerator0
    Override Seed=false
    Shape Origin=$Output.SocketLocation.SampledPosition
    Shape Primitive=::NewEnumerator2
    Spawn Group:NiagaraInt32=&SpawnGroup()
    Spawn Group Mask=2
    Surface Only=false
    UseOwnerScale=true
   )
   /Niagara/Modules/Update/SubUV/SubUVAnimation(SubUV Animation Mode=::NewEnumerator1)
   &AddVelocityFromPoint(
    Module.Use Falloff Distance=false
    Origin Offset=&MergeHelper_VectorSwitch(0=<-4000 0 400>, 1=<4000 0 100>, 2=<0 0 0>, 3=?, Index=&SpawnGroup(), Num=::NewEnumerator1)
    Origin Offset Coordinate Space=::Simulation
    Velocity Strength=30
   )
   &CameraOffset(Camera Offset Amount=&MaskFloatBySpawnGroup(Masked Float=10, Passthrough Float=5, Spawn Group Mask=2))
   &AddVelocity(
    Velocity=&MaskVector(
     Mask Source=0
     Mask Parameter=?
     Mask=2
     Masked Vector=&FN_FrontEndVector3Modifier(
      Gameplay=&MakeCustomVectorFromBool(
       Bool=&SetBoolByFloatComparison(A=&VectorLength(Vector=$User.OwnerLocalVelocity), B=50)
       True Vector=<0 -15 10>
       False Vector=<0 -30 10>
      )
      Front End=<0 -30 10>
     )
     Passthrough Vector=&RandomRangeVector(Minimum=<5 -70 10>, Maximum=<-5 -30 20>)
    )
    Velocity Speed Scale=5
   )
  ]
  particleUpdateScript [
   &ParticleState()
   &SpriteRotationRate(Rotation Rate=&RandomRangeFloat(Minimum=-90, Maximum=90))
   &Drag(Drag=3, Use Linear Drag=true, Use Rotational Drag=false)
   &GravityForce(Gravity=&MaskVector(Mask Source=0, Mask Parameter=?, Mask=2, Masked Vector=<0 0 -300>, Passthrough Vector=<0 0 -600>))
   &SolveForcesAndVelocity(Clamp Velocity=false, Limit Acceleration=false)
   &ScaleSpriteSize(Scale Factor=/Niagara/DynamicInputs/Helpers/MaskVector2d(
     Mask Source:NiagaraInt32=0
     Mask Parameter:NiagaraInt32=?
     Mask:NiagaraInt32=2
     Masked Vector2D:Vector2f=&Multiply_Vector2D_ByFloat(
      Float=&FloatFromCurve(
       FloatCurve=DataInterface() {propertyValues {Curve {keys [
           {arriveTangent=-0.2121, interpMode=RCIM_Cubic, leaveTangent=-0.2121, tangentMode=RCTM_User, time=0, value=0.33}
           {interpMode=RCIM_Cubic, tangentMode=RCTM_User, time=1, value=1}
          ]}}}
       Scale Curve=0.5
      )
      Vector2D=&Vector2DFromCurve(
       Vector2Curve=DataInterface() {propertyValues {XCurve {keys [<at 0 1>,<at 0.5179 1.2321>]}, YCurve {keys [<at 0 1>,<at 0.5238 0.7531>]}}}
       ScaleCurve=<1.5 1>
      )
     )
     Passthrough Vector2D:Vector2f=&Vector2DFromFloat(Value=&FloatFromCurve(
       FloatCurve=DataInterface() {propertyValues {Curve {keys [
           {arriveTangent=-0.2121, interpMode=RCIM_Cubic, leaveTangent=-0.2121, tangentMode=RCTM_User, time=0, value=0.33}
           {interpMode=RCIM_Cubic, tangentMode=RCTM_User, time=1, value=1}
          ]}}}
       Scale Curve=0.5
      ))
    ))
   &DynamicMaterialParameters(
    Index 0 Param 1=&RandomRangeFloat(Minimum=0, Maximum=1)
    Index 0 Param 3=&FloatFromCurve(FloatCurve=DataInterface() {propertyValues {Curve {keys [{interpMode=RCIM_Cubic, tangentMode=RCTM_User, time=0, value=0}, {interpMode=RCIM_Cubic, tangentMode=RCTM_User, time=1, value=1}]}}})
    Index 0 Param 4=&FloatFromCurve(FloatCurve=DataInterface() {propertyValues {Curve {keys [<at 0 0 RCIM_Cubic>, {arriveTangent=1.5276, interpMode=RCIM_Cubic, leaveTangent=1.5276, tangentMode=RCTM_User, time=1, value=1}]}}})
    Module.Index0Float0Write=true
    Module.Index0Float1Write=true
    Module.Index0Float2Write=true
    Module.Index0Float3Write=true
   )
   &ScaleColor(
    Scale Alpha=&FloatFromCurve(
     FloatCurve=DataInterface() {propertyValues {Curve {keys [
         <at 0 0 RCIM_Cubic>
         {arriveTangent=0.6013, interpMode=RCIM_Cubic, leaveTangent=0.6013, time=0.1776, value=1}
         {arriveTangent=-1.2159, interpMode=RCIM_Cubic, leaveTangent=-1.2159, time=0.4636, value=0.2788}
         <at 1 0 RCIM_Cubic>
        ]}}}
     Scale Curve=&FN_FrontEndFloatModifier(
      Gameplay=0.5
      Front End=&MaskFloat(Mask Source=0, Mask Parameter=?, Mask=0, Masked Float=1.5, Passthrough Float=0.8)
     )
    )
    ScaleA=true
    ScaleRGB=false
   )
   &BlendSpace("Position Blend (World <-> Local)":NiagaraFloat=&FN_FrontEndFloatModifier(
     Gameplay=&MaskFloat(
      Mask Source=0
      Mask Parameter=?
      Mask=&SpawnGroup()
      Masked Float=1
      Passthrough Float=&MakeCustomFloatFromBool(
       Bool=&SetBoolByFloatComparison(A=&VectorLength(Vector=$Engine.Owner.Velocity), B=350)
       True Float=0
       False Float=&FloatFromCurve(FloatCurve=DataInterface() {propertyValues {Curve {keys [<at 0.1106 1 RCIM_Cubic>,<at 0.5991 0.7 RCIM_Cubic>]}}})
      )
     )
     Front End=1
     Translation Contribution:NiagaraFloat=1
     Rotation Contribution:NiagaraFloat=0.5
    ))
  ]
  rendererClasses [&NiagaraSpriteRendererProperties]
  renderers [{rendererClass=&NiagaraSpriteRendererProperties}]
 }
 Trail_Smoke_Soft_Top {
  emitterUpdateScript [
   &EmitterState(Life Cycle Mode=::NewEnumerator1, Loop Behavior=::NewEnumerator1, Loop Delay=0.03, Loop Duration Mode=::NewEnumerator1, UseLoopDelay=false)
   &SpawnRate(
    Spawn Probability=&MakeCustomFloatFromBool(
     Bool=&SetBoolByFloatComparison(A=&VectorLength(Vector=$User.OwnerLocalVelocity), B=50)
     True Float=1
     False Float=0
    )
    SpawnGroup=1
    SpawnRate=45
    Use Spawn Probability=false
   ) {module=SpawnRate001}
   &SpawnRate(
    Spawn Probability=&MakeCustomFloatFromBool(
     Bool=&SetBoolByFloatComparison(A=&VectorLength(Vector=$User.OwnerLocalVelocity), B=50)
     True Float=1
     False Float=0
    )
    SpawnRate=45
    Use Spawn Probability=false
   )
  ]
  particleSpawnScript [
   /SetVariables(Particles.RibbonID=&MakeRibbonIDFromSpawnGroup()) {isSetParameters=true}
   &InitializeRibbon(
    Lifetime=&FN_FrontEndFloatModifier(Gameplay=1.5, Front End=1.6)
    Module.Write Color=true
    Module.Write Lifetime=true
    Module.Write Mass=false
    Module.Write Material Random=true
    Module.Write Position=false
    Module.Write RibbonTwist=false
    Module.Write RibbonWidth=true
    Ribbon Width=30
   )
   &SocketLocation(
    SamplePosOffset=true
    Sampled Position Offset=&MaskVector(Mask Source=0, Mask Parameter=?, Mask=0, Masked Vector=<8 0 -10>, Passthrough Vector=<-8 0 -10>)
    Socket Index=&SpawnGroup()
    Socket Reader=$System.Socket Reader
    Socket Sampling Mode=::NewEnumerator3
    VelocitySpeedLimit=false
   )
   /SetVariables(Particles.RibbonFacing=$Engine.Owner.SystemZAxis) {isSetParameters=true}
   &CameraOffset(Camera Offset Amount=15)
   &AddVelocity(
    Velocity=<0 -60 0>
    Velocity Speed Scale=&FN_FrontEndFloatModifier(
     Gameplay=&MakeCustomFloatFromBool(
      Bool=&Boolean_OrOperation(
       Bool1=$User.bIsFrontEndPreview
       Bool2=&SetBoolByFloatComparison(A=&VectorLength(Vector=$Engine.Owner.Velocity), B=50, Comparison Type=::NewEnumerator6)
      )
      True Float=0.5
      False Float=1
     )
     Front End=1
    )
   )
  ]
  particleUpdateScript [
   &ParticleState()
   &SolveForcesAndVelocity(Clamp Velocity=false, Limit Acceleration=false)
   &ScaleColor(
    Linear Color Curve=DataInterface() {propertyValues {
      AlphaCurve {keys [<at 0 3 RCIM_Cubic>,<at 0.2653 1>,<at 1 0 RCIM_Cubic>]}
      BlueCurve {keys [{arriveTangent=-52.6135, interpMode=RCIM_Cubic, time=0.039, value=0.6285}, <at 0.5737 0.8456>, <at 1 1>]}
      GreenCurve {keys [{arriveTangent=-38.4858, interpMode=RCIM_Cubic, time=0.039, value=1}, <at 0.5737 0.133>, <at 1 0.494>]}
      RedCurve {keys [<at 0.039 0 RCIM_Cubic>,<at 0.5737 0.3118>,<at 1 0.9197>]}
     }}
    Scale Mode=::NewEnumerator2
   )
   &ScaleColor(
    Scale RGB=&Multiply_Vector(
     A=&VectorFromFloat(Value=&FNTimeOfDayScalar(Night Scale=0.5, Morning Scale=0.25, Evening Scale=0.5))
     B=&FN_Select_Vector_from_Time_Of_Day(Morning=<1 1 1>, Day=<1 1 1>, Evening=<1 1 1>, Night=<0.25 1 1>)
    )
    ScaleA=true
    ScaleRGB=true
   ) {module=ScaleColor001}
   &ScaleRibbonWidth(Ribbon Width Scale=&FloatFromCurve(FloatCurve=DataInterface() {propertyValues {Curve {keys [{arriveTangent=-1.2963, interpMode=RCIM_Cubic, leaveTangent=-1.2963, tangentMode=RCTM_User, time=0, value=2}, <at 1 0.8781>]}}}))
   &DynamicMaterialParameters(
    Index 0 Param 1=&FloatFromCurve(FloatCurve=DataInterface() {propertyValues {Curve {keys [<at 0 1>,<at 1 0>]}}})
    Index 0 Param 2=&RandomRangeFloat(Minimum=0, Maximum=1)
    Index 0 Param 4=&FloatFromCurve(FloatCurve=DataInterface() {propertyValues {Curve {keys [
         {interpMode=RCIM_Cubic, tangentMode=RCTM_User, time=0, value=1}
         {arriveTangent=-0.6219, interpMode=RCIM_Cubic, leaveTangent=-0.6219, tangentMode=RCTM_User, time=1, value=0}
        ]}}})
    Module.Param0WriteEnabled=true
    Module.Param1WriteEnabled=true
    Module.Param2WriteEnabled=true
    Module.Param3WriteEnabled=true
   )
   &BlendSpace("Position Blend (World <-> Local)":NiagaraFloat=&FN_FrontEndFloatModifier(
     Gameplay=&MakeCustomFloatFromBool(
      Bool=&SetBoolByFloatComparison(A=&VectorLength(Vector=$Engine.Owner.Velocity), B=400)
      True Float=1
      False Float=&FloatFromCurve(FloatCurve=DataInterface() {propertyValues {Curve {keys [<at 0.2074 1>,<at 1 0>]}}})
     )
     Front End=1
     Translation Contribution:NiagaraFloat=1
     Rotation Contribution:NiagaraFloat=1
    ))
  ]
  rendererClasses [&NiagaraRibbonRendererProperties]
  renderers [{rendererClass=&NiagaraRibbonRendererProperties}]
 }
 side_feet_clumps {
  emitterSpawnScript [/SetVariables(Emitter.FE Particle Spawn Rate=35, Emitter.Idle Spawn Rate=20, Emitter.Max Traversal Speed=150, Emitter.Particle Spawn Rate=30) {isSetParameters=true}]
  emitterUpdateScript [
   &EmitterState()
   &SpawnRate(
    SpawnRate=&FN_FrontEndFloatModifier(
     Gameplay=&MakeCustomFloatFromBool(
      Bool=&SetBoolByFloatComparison(A=&VectorLength(Vector=$User.OwnerLocalVelocity), B=100)
      True Float=$Emitter.Particle Spawn Rate
      False Float=$Emitter.Idle Spawn Rate
     )
     Front End=$Emitter.FE Particle Spawn Rate
    )
    Use Spawn Probability=false
   )
  ]
  particleSpawnScript [
   &PartitionParticles()
   /SetVariables(Particles.InitSpriteRotRate=&RandomRangeFloat(Minimum=-2, Maximum=2)) {isSetParameters=true}
   &InitializeParticle(
    Lifetime Max=0.5
    Lifetime Min=0.35
    Lifetime Mode=::NewEnumerator1
    Sprite Rotation Angle=&RandomRangeFloat(Minimum=0, Maximum=360)
    Sprite Rotation Mode=::NewEnumerator0
    Sprite Size=&Multiply_Vector2D(
     A=&FN_FrontEndVector2Modifier(
      Front End:Vector2f=<1 1>
      Gameplay:Vector2f=&MakeCustomVector2DFromBool(
       Bool=&SetBoolByFloatComparison(A=&VectorLength(Vector=$Engine.Owner.Velocity), B=50, Comparison Type=::NewEnumerator6)
       True Vector2D=<1 0.5>
       False Vector2D=<1 1>
      )
     )
     B=&RandomRangeVector2D(Minimum=<20 30>, Maximum=<30 20>)
    )
    Sprite Size Mode=::NewEnumerator4
    Sprite UV Mode=::NewEnumerator2
    UsePositionOffset=false
    Write Material Random=true
   )
   &SocketLocation(
    SamplePosOffset=true
    Sampled Position Offset=&MaskVector(Mask Source=1, Mask Parameter=$Particles.PartitionParticles.Partition, Mask=0, Masked Vector=<5 10 -15>, Passthrough Vector=<-5 10 -15>)
    Sampled Position Offset Coordinate Space:ENiagaraSkelMeshTransforms=::NewEnumerator2
    Socket Index=$Particles.PartitionParticles.Partition
    Socket Reader=$System.Socket Reader
    Socket Sampling Mode=::NewEnumerator3
    VelocitySpeedLimit=false
   )
   &SphereLocation(
    MaskSpawnGroup=false
    Non Uniform Scale=<1 1 0.2>
    Override Seed=false
    Sphere Coordinate Space=::Simulation
    Sphere Radius=7
    Surface Only=false
    Surface Only Band Thickness=1
    UseNonUniformScale=true
    UseOffset=false
    UseSphereOrientation=false
   )
   /SetVariables(Particles.Lifetime=&MakeCustomFloatFromBool(
     Bool=&SetBoolByFloatComparison(
      A=50
      B=&DistanceBetweenPositions(
       A=&MakeVector(
        X=0
        Y=0
        Z=&MakeFloatFromVector(Vector=Linked() {linkedVariable {name=Particles.SkeletalMeshLocation001.SampledPosition, type=NiagaraPosition}}, Channel=::NewEnumerator2)
       )
       B=&MakeVector(
        X=0
        Y=0
        Z=&MakeFloatFromVector(Vector=Linked() {linkedVariable {name=Particles.SkeletalMeshLocation.SampledPosition, type=NiagaraPosition}}, Channel=::NewEnumerator2)
       )
      )
     )
     True Float=$Particles.Lifetime
     False Float=0
    )) {isSetParameters=true}
   &AddVelocity(
    Velocity=&MergeHelper_VectorSwitch(0=<-0.2 -2.5 0.5>, 1=<0.2 -2.5 0.5>, 2=?, 3=?, Index=$Particles.PartitionParticles.Partition, Num=::NewEnumerator0)
    Velocity Speed Scale=&RandomRangeFloat(Minimum=25, Maximum=70)
   ) {module=AddVelocity001}
  ]
  particleUpdateScript [
   &ParticleState()
   &Drag(Drag=&RandomRangeFloat(Minimum=0.8, Maximum=1.2), Use Linear Drag=true, Use Rotational Drag=false)
   &SolveForcesAndVelocity(Clamp Velocity=false, Limit Acceleration=false)
   &ScaleSpriteSizeBySpeed(Max Scale Factor=<1.5 2.5>, Sample Scale Factor By Curve=true, Velocity Threshold=150)
   &SpriteRotationRate(Rotation Rate=&Multiply_Float(
     A=$Particles.InitSpriteRotRate
     B=&FloatFromCurve(
      FloatCurve=DataInterface() {propertyValues {Curve {keys [
          {arriveTangent=-4.0107, interpMode=RCIM_Cubic, leaveTangent=-4.0107, tangentMode=RCTM_User, time=-0.0046, value=1.0113}
          {arriveTangent=-0.2496, interpMode=RCIM_Cubic, leaveTangent=-0.2496, tangentMode=RCTM_User, time=0.5, value=0.16}
         ]}}}
      CurveIndex=$Emitter.Age
      Scale Curve=150
     )
    ))
   &ScaleColor(
    Scale Alpha=&RandomRangeFloat(Minimum=0.05, Maximum=0.1)
    Scale RGB=&MakeVectorFromLinearColorRGB(Color=&MergeHelper_ColorSwitch(
      0=&ColorFromCurve(ColorCurve=DataInterface() {propertyValues {
         AlphaCurve {keys [<at 0 0>,<at 0.0976 1 RCIM_Cubic>,<at 1 0 RCIM_Cubic>]}
         BlueCurve {keys [
           {arriveTangent=-4.4608, interpMode=RCIM_Cubic, time=0, value=1.5696}
           {arriveTangent=-1.2803, interpMode=RCIM_Cubic, leaveTangent=-1.2803, time=0.3516, value=0.7507}
           <at 0.5762 0.8318>
          ]}
         GreenCurve {keys [
           {arriveTangent=-5.6888, interpMode=RCIM_Cubic, time=0, value=2.3}
           {arriveTangent=-3.9172, interpMode=RCIM_Cubic, leaveTangent=-3.9172, time=0.3516, value=1.1}
           <at 0.5762 0.0428>
          ]}
         RedCurve {keys [
           {arriveTangent=-1.8222, interpMode=RCIM_Cubic, time=0, value=0}
           {arriveTangent=1.7355, interpMode=RCIM_Cubic, leaveTangent=1.7355, time=0.3516, value=0}
           <at 0.5762 1>
          ]}
        }})
      1=&ColorFromCurve(ColorCurve=DataInterface() {propertyValues {
         AlphaCurve {keys [<at 0 0>,<at 0.0884 1 RCIM_Cubic>,<at 1 0 RCIM_Cubic>]}
         BlueCurve {keys [
           {arriveTangent=-6.1128, interpMode=RCIM_Cubic, time=0, value=4}
           {arriveTangent=-6.8285, interpMode=RCIM_Cubic, leaveTangent=-6.8285, time=0.3248, value=0.2962}
           <at 0.4631 0.8374>
          ]}
         GreenCurve {keys [
           {arriveTangent=-8.0251, interpMode=RCIM_Cubic, time=0, value=3.5164}
           {arriveTangent=-7.5422, interpMode=RCIM_Cubic, leaveTangent=-7.5422, time=0.3248, value=0.434}
           <at 0.4631 0.0232>
          ]}
         RedCurve {keys [
           {arriveTangent=-2.0036, interpMode=RCIM_Cubic, time=0, value=0}
           {arriveTangent=0.7926, interpMode=RCIM_Cubic, leaveTangent=0.7926, time=0.3248, value=0}
           <at 0.4631 0.3671>
          ]}
        }})
      2=&ColorFromCurve(ColorCurve=DataInterface() {propertyValues {
         AlphaCurve {keys [<at 0 0>,<at 0.1253 1 RCIM_Cubic>,<at 1 0 RCIM_Cubic>]}
         BlueCurve {keys [{arriveTangent=-2.5516, interpMode=RCIM_Cubic, time=0, value=1.3665}, <at 0.7988 0.2124 RCIM_Cubic>]}
         GreenCurve {keys [{arriveTangent=-3.4924, interpMode=RCIM_Cubic, time=0, value=2.3}, <at 0.7988 0.2101 RCIM_Cubic>]}
         RedCurve {keys [{arriveTangent=-0.8019, interpMode=RCIM_Cubic, time=0, value=0}, <at 0.7988 0 RCIM_Cubic>]}
        }})
      3=?
      Index=/CascadeToNiagaraConverter/NiagaraScripts/DynamicInputScripts/SequentialInt(Use Array Data:NiagaraBool=false, First Number:NiagaraInt32=0, Sequence:NiagaraDataInterfaceArrayInt32=?, Last Number:NiagaraInt32=2)
      Num=::NewEnumerator1
     ))
    ScaleA=true
    ScaleRGB=true
   ) {module=ScaleColor001}
   &ScaleColor(
    Scale Alpha=&FloatFromCurve(
     FloatCurve=DataInterface() {propertyValues {Curve {
        defaultValue=1
        keys [
         {interpMode=RCIM_Cubic, tangentMode=RCTM_User, time=0, value=0.01}
         {interpMode=RCIM_Cubic, tangentMode=RCTM_User, time=0.5, value=1}
         {tangentMode=RCTM_User, time=1, value=0.01}
        ]
       }}}
     CurveIndex=Linked() {linkedVariable {name=NPC.FortniteNPC.RemappedTimeOfDay_0-1, type=NiagaraFloat}}
    )
    Scale RGB=&MakeVectorFromLinearColorRGB(Color=&ColorFromCurve(
      ColorCurve=DataInterface() {propertyValues {
        AlphaCurve {keys [<at 0 1 RCIM_Cubic>]}
        BlueCurve {keys [
          <at 0.0043 0.7>
          {arriveTangent=-0.1359, interpMode=RCIM_Cubic, leaveTangent=-0.1359, time=0.1789, value=0.7}
          <at 0.2812 0.6624>
          {arriveTangent=1.2198, interpMode=RCIM_Cubic, leaveTangent=1.2198, time=0.3527, value=1}
          {arriveTangent=-1.0217, interpMode=RCIM_Cubic, leaveTangent=-1.0217, time=0.558, value=1}
          {arriveTangent=-0.9848, interpMode=RCIM_Cubic, leaveTangent=-0.9848, time=0.8029, value=0.54}
          <at 0.8615 0.7011>
          <at 1 0.7>
         ]}
        GreenCurve {keys [
          <at 0.0043 0.4211>
          {arriveTangent=0.9841, interpMode=RCIM_Cubic, leaveTangent=0.9841, time=0.1789, value=0.5785}
          <at 0.2812 0.6936>
          {arriveTangent=1.1069, interpMode=RCIM_Cubic, leaveTangent=1.1069, time=0.3527, value=1}
          {arriveTangent=-0.8536, interpMode=RCIM_Cubic, leaveTangent=-0.8536, time=0.558, value=1}
          {arriveTangent=-1.3917, interpMode=RCIM_Cubic, leaveTangent=-1.3917, time=0.8029, value=0.6157}
          <at 0.8615 0.5776>
          <at 1 0.4138>
         ]}
        RedCurve {keys [
          <at 0.0043 0.3653>
          {arriveTangent=1.9379, interpMode=RCIM_Cubic, leaveTangent=1.9379, time=0.1789, value=0.5542}
          <at 0.2812 0.902>
          {arriveTangent=0.3542, interpMode=RCIM_Cubic, leaveTangent=0.3542, time=0.3527, value=1}
          {arriveTangent=-0.2221, interpMode=RCIM_Cubic, leaveTangent=-0.2221, time=0.558, value=1}
          {arriveTangent=-1.476, interpMode=RCIM_Cubic, leaveTangent=-1.476, time=0.8029, value=0.9}
          <at 0.8615 0.552>
          <at 1 0.3566>
         ]}
       }}
      CurveIndex=Linked() {linkedVariable {name=NPC.FortniteNPC.RemappedTimeOfDay_0-1, type=NiagaraFloat}}
     ))
    ScaleA=false
    ScaleRGB=true
   )
   &DynamicMaterialParameters(
    Index 0 Param 1=$Particles.NormalizedAge
    Module.Index0Float0Write=true
    Module.Index0Float1Write=true
    Module.Index0Float2Write=true
    Module.Index0Float3Write=true
   )
   &BlendSpace("Position Blend (World <-> Local)":NiagaraFloat=&FN_FrontEndFloatModifier(
     Gameplay=&MakeCustomFloatFromBool(
      Bool=&SetBoolByFloatComparison(A=&VectorLength(Vector=$Engine.Owner.Velocity), B=400)
      True Float=1
      False Float=&FloatFromCurve(FloatCurve=DataInterface() {propertyValues {Curve {keys [<at 0 1 RCIM_Cubic>, {arriveTangent=-3.5186, interpMode=RCIM_Cubic, leaveTangent=-3.5186, tangentMode=RCTM_User, time=0.5, value=0}]}}})
     )
     Front End=1
     Translation Contribution:NiagaraFloat=1
     Rotation Contribution:NiagaraFloat=1
    ))
  ]
  rendererClasses [&NiagaraSpriteRendererProperties]
  renderers [{rendererClass=&NiagaraSpriteRendererProperties}]
 }
}
summary {
 systemName=NS_Emote_RainbowRaid_Traversal
 emitters [
  {emitterName=Trail_Smoke_Soft_Top, rendererClasses [&NiagaraRibbonRendererProperties]}
  {emitterName=Auroras, rendererClasses [&NiagaraMeshRendererProperties]}
  {emitterName=side_feet_clumps, rendererClasses [&NiagaraSpriteRendererProperties]}
  {emitterName=Splashes, rendererClasses [&NiagaraSpriteRendererProperties]}
  {emitterName=Core, rendererClasses [&NiagaraRibbonRendererProperties]}
  {bEnabled=false, emitterName=Auroras001, rendererClasses [&NiagaraMeshRendererProperties]}
  {bEnabled=false, emitterName=Auroras002, rendererClasses [&NiagaraMeshRendererProperties]}
 ]
 userVariables [
  User.bIsFrontEnd {type=NiagaraBool}
  User.bIsFrontEndPreview {type=NiagaraBool}
  User.OwnerLocalVelocity {type=Vector3f, description="", defaultValue {struct=Vector3f, value=<0 0 0>}}
 ]
}
